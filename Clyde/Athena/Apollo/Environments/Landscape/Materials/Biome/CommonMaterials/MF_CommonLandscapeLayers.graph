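# Material function that builds the common landscape layer stack of grass, gravel, mud, sand and asphalt, with biome color inputs.
# Grass is the base, receiving a combined mask of the other layers, and MF_BlendLandscapeLayers chains each layer on by landscape weight.
# Mud reuses gravel's texture outputs, and grass noise drives gravel and mud variance.
# Outputs include blended material, heightmap with a feature-level zero-out term, and grass and gravel intermediates.
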MaterialFunction /Clyde/Athena/Apollo/Environments/Landscape/Materials/Biome/CommonMaterials/MF_CommonLandscapeLayers
graph {nodes [
  n1 = FunctionInput(Preview=Constant()) {bUsePreviewValueAsDefault=true, inputName=BiomeMask_1, inputType=FunctionInput_Scalar, sortPriority=-9}
  n2 = FunctionInput(Preview=StaticBool()) {
   bUsePreviewValueAsDefault=true
   inputName=UseBiomeColorMask_1
   inputType=FunctionInput_StaticBool
   sortPriority=-10
   previewValue=<64 64 64 6>
  }
  n3 = FunctionInput() {bUsePreviewValueAsDefault=true, inputName=HeightmapFromCliff_1, inputType=FunctionInput_Scalar, sortPriority=-7, previewValue=<1 1 1 1>}
  n4 = Max(A=LandscapeLayerSample() {parameterName=Mud}, B=LandscapeLayerSample() {parameterName=Gravel})
  n5 = LandscapeLayerSample() {parameterName=Sand}
  n6 = Max(
   A=Max(A=Max(A=@n4, B=@n5), B=LandscapeLayerSample() {parameterName=Asphalt})
   B=FunctionInput(Preview=Constant()) {bUsePreviewValueAsDefault=true, inputName=GrassBorderColorMask_1, inputType=FunctionInput_Scalar, sortPriority=-8}
  )
  n7 = Constant() {r=2}
  n8 = MaterialFunctionCall!(
   BiomeMask=@n1
   UseBiomeColorMask=@n2
   CliffMask=@n3
   Grass-EdgeBiomeColorA=FunctionInput() {bUsePreviewValueAsDefault=true, inputName=Grass-EdgeBiomeColorA_1}
   Grass-EdgeBiomeColorB=FunctionInput() {bUsePreviewValueAsDefault=true, inputName=Grass-EdgeBiomeColorB_1}
   CombinedLandscapeLayersMask=Saturate(None=Multiply(A=@n6, B=@n7))
  ) {materialFunction=&/MF_Common_Grass}
  n9 = LandscapeLayerSample() {parameterName=Gravel}
  n10 = FunctionInput() {bUsePreviewValueAsDefault=true, inputName=GravelBiomeColorDiv_1, sortPriority=7, previewValue=<1 1 1 1>}
  n11 = FunctionInput() {bUsePreviewValueAsDefault=true, inputName=GravelBiomeColor_1, sortPriority=8, previewValue=<1 1 1 1>}
  n12 = MaterialFunctionCall!(
   UseBiomeColorMask=@n2
   UseLandscapeCoords
   VarianceMask=@n8.GrassColorMaskA
   LayerWeight=@n9
   BiomeColorMask=@n1
   BiomeColorDiv=@n10
   BiomeColor=@n11
  ) {materialFunction=&/MF_Common_Gravel}
  n13 = MaterialFunctionCall!(
   Material 1=@n8.Material
   Height Texture 1=@n8.Material
   Material 2=@n12.Material
   Height Texture 2=@n12.Material
   Transistion Phase=@n9
   UseContrast
   Contrast
  ) {materialFunction=&/MF_BlendLandscapeLayers}
  n14 = FunctionInput() {bUsePreviewValueAsDefault=true, inputName=MudBiomeColorDiv_1, sortPriority=9, previewValue=<1 1 1 1>}
  n15 = FunctionInput() {bUsePreviewValueAsDefault=true, inputName=MudBiomeColor_1, sortPriority=10, previewValue=<1 1 1 1>}
  n16 = MaterialFunctionCall!(
   Diffuse=FunctionOutput(None=@n12.Diffuse) {outputName=GravelDiffuse_1, sortPriority=3}
   Height=FunctionOutput(None=@n12.Height) {outputName=GravelHeight_1, sortPriority=2}
   Rough=FunctionOutput(None=@n12.Rough) {outputName=GravelRough_1, sortPriority=5}
   VarianceMask=Multiply(A=@n8.GrassColorMaskA) {constB=0.5, desc=Makes it looks a little more like OG reducing variance}
   UseBiomeColorMask=@n2
   Spec=FunctionOutput(None=@n12.Spec) {outputName=GravelSpec_1, sortPriority=4}
   Normals=FunctionOutput(None=@n12.Normals) {outputName=GravelNormals_1, sortPriority=6}
   BiomeColorMask=@n1
   BiomeColorDiv=@n14
   BiomeColor=@n15
   BiomeRoughMod=FunctionInput(Preview=Constant() {r=1}) {bUsePreviewValueAsDefault=true, inputName=MudBiomeRoughMod_1, inputType=FunctionInput_Scalar, sortPriority=11}
  ) {materialFunction=&/MF_Common_Mud}
  n17 = MaterialFunctionCall(
   Material 1=@n13.BlendedMaterial
   Height Texture 1=@n13.BlendedMaterial
   Material 2=@n16.Material
   Height Texture 2=@n16.Material
   Transistion Phase=LandscapeLayerSample() {parameterName=Mud}
  ) {materialFunction=&MF_BlendLandscapeLayers}
  n18 = FunctionInput() {bUsePreviewValueAsDefault=true, inputName=SandBiomeColorDiv_1, sortPriority=12, previewValue=<1 1 1 1>}
  n19 = FunctionInput() {bUsePreviewValueAsDefault=true, inputName=SandBiomeColor_1, sortPriority=13, previewValue=<1 1 1 1>}
  n20 = MaterialFunctionCall!(UseBiomeColorMask=@n2, BiomeColorMask=@n1, BiomeColorDiv=@n18, BiomeColor=@n19, UseUnderOceanSandMaterial) {materialFunction=&/MF_Common_Sand}
  n21 = MaterialFunctionCall(
   Material 1=@n17.BlendedMaterial
   Height Texture 1=@n17.BlendedMaterial
   Material 2=@n20.Material
   Height Texture 2=@n20.Material
   Transistion Phase=LandscapeLayerSample() {parameterName=Sand}
  ) {materialFunction=&MF_BlendLandscapeLayers}
  n22 = MaterialFunctionCall() {materialFunction=&/MF_Common_Asphalt}
  n23 = MaterialFunctionCall(
   Material 1=@n21.BlendedMaterial
   Height Texture 1=@n21.BlendedMaterial
   Material 2=@n22.Material
   Height Texture 2=@n22.Material
   Transistion Phase=LandscapeLayerSample() {parameterName=Asphalt}
  ) {materialFunction=&MF_BlendLandscapeLayers}
  FunctionOutput(None=@n23.BlendedMaterial) {outputName=Material, sortPriority=-1}
  n24 = Multiply(A=MaterialFunctionCall() {materialFunction=&/MF_Common_Landscape_Layers}.Result) {constB=0}
  n25 = ScalarParameter() {parameterName=FeatureZeroOut}
  FunctionOutput(None=Add(A=@n23.Resulting Height Map, B=FeatureLevelSwitch(Default=@n24, ES2=@n25, ES3_1=@n25, SM4_REMOVED=@n24, SM5=@n24))) {outputName=Heightmap}
  FunctionOutput(None=@n8.Heightmap) {outputName=GrassHeightmap_1, sortPriority=7}
  FunctionOutput(None=@n8.GrassColorMaskC) {outputName=GrassNoiseMask_1, sortPriority=9}
  FunctionOutput(None=@n8.Normals) {outputName=GrassNormal_1, sortPriority=8}
  FunctionOutput(None=@n8.BasecolorDetail) {outputName=GrassBaseColorDetail_1, sortPriority=10}
 ]}
