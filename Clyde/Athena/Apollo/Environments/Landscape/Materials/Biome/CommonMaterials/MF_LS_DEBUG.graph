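# MF_LS_DEBUG is a material function that adds debug views to the base color of Apollo landscape biome materials.
# It blends in a grid from MF_Apollo_LS_GridDebug, multiplies by a biome color input, then can replace base color with an arbitrary debug input.
# Scalar parameters in the Debug group control each stage.
# The UseDebug static switch outputs either this chain or the untouched input material attributes.
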MaterialFunction /Clyde/Athena/Apollo/Environments/Landscape/Materials/Biome/CommonMaterials/MF_LS_DEBUG
graph {nodes [
  n1 = FunctionInput() {inputName=Material, inputType=FunctionInput_MaterialAttributes}
  n2 = GetMaterialAttributes(None=@n1) {attributeGetTypes [#1]}
  n3 = MaterialFunctionCall() {materialFunction=/Game/Athena/Apollo/Environments/Landscape/Materials/MaterialFunctions/MF_Apollo_LS_GridDebug}
  n4 = ScalarParameter() {group=Debug, parameterName=DEBUG-ShowGrid, sliderMax=1}
  n5 = SetMaterialAttributes!(
   MaterialAttributes=@n2.MaterialAttributes
   BaseColor=LinearInterpolate(A=@n2.BaseColor, B=LinearInterpolate(A=@n2.BaseColor, B=@n3.RGB, Alpha=@n3.RGB), Alpha=Saturate(None=@n4))
  ) {attributeSetTypes [#1]}
  n6 = MaterialFunctionCall!(Material=@n5) {materialFunction=/Engine/Functions/MaterialLayerFunctions/MatLayerBlend_BreakBaseColor}
  n7 = FunctionInput() {inputName=DebugBiomeColor, sortPriority=1, previewValue=<1 1 1 1>}
  n8 = SetMaterialAttributes(
   MaterialAttributes=@n5
   BaseColor=LinearInterpolate(A=@n6.BaseColor, B=Multiply(A=@n6.BaseColor, B=@n7), Alpha=ScalarParameter() {group=Debug, parameterName=DEBUG-BiomeColor})
  ) {attributeSetTypes [#1]}
  n9 = FunctionInput(Preview=Constant3Vector()) {bUsePreviewValueAsDefault=true, inputName=DebugAsBasecolor, sortPriority=1, previewValue=<1 1 1 1>}
  n10 = LinearInterpolate(
   A=MaterialFunctionCall(Material=@n8) {materialFunction=&MatLayerBlend_BreakBaseColor}.BaseColor
   B=@n9
   Alpha=ScalarParameter() {group=Debug, parameterName=DEBUG-InputAsBasecolor}
  )
  FunctionOutput(None=StaticSwitchParameter(True=SetMaterialAttributes(MaterialAttributes=@n8, BaseColor=@n10) {attributeSetTypes [#1]}, False=@n1) {group=Debug, parameterName=UseDebug})
 ]}
