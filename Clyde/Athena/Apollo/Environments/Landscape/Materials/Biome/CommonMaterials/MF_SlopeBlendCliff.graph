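# Landscape material function that blends an incoming biome material with the shared MF_Common_Cliff material by terrain slope, with optional biome color inputs.
# Slope from absolute vertex-normal Z drives a two-heightmap height lerp between the input and cliff heights, whose alpha blends the material attributes.
# It also outputs heightmaps for snow, grass and regular use, plus grass mesh and border color masks.
# The binary grass mesh mask uses simpler slope-only math on forward and mobile shading paths.
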MaterialFunction /Clyde/Athena/Apollo/Environments/Landscape/Materials/Biome/CommonMaterials/MF_SlopeBlendCliff
graph {nodes [
  Constant() {r=1.4}
  Constant() {r=-1}
  n1 = FunctionInput(Preview=StaticBool()) {bUsePreviewValueAsDefault=true, inputName=UseBiomeColorMask, inputType=FunctionInput_StaticBool, previewValue=<64 64 64 6>}
  n2 = FunctionInput() {bUsePreviewValueAsDefault=true, inputName=BiomeColorDiv, sortPriority=2, previewValue=<1 1 1 1>}
  n3 = Abs(None=ComponentMask(None=VertexNormalWS()) {b=true})
  n4 = MaterialFunctionCall!(
   UseBiomeColorMask=@n1
   BiomeColorMask=FunctionInput(Preview=Constant() {r=1}) {bUsePreviewValueAsDefault=true, inputName=BiomeColorMask, inputType=FunctionInput_Scalar, sortPriority=-5}
   BiomeColorDiv=@n2
   BiomeColor=FunctionInput() {bUsePreviewValueAsDefault=true, inputName=BiomeColor, sortPriority=3, previewValue=<1 1 1 1>}
   Slope=@n3
  ) {materialFunction=&/MF_Common_Cliff}
  n5 = FunctionInput() {inputName=Heightmap, sortPriority=-1}
  n6 = Saturate(None=Multiply(A=OneMinus(None=@n3), B=Constant() {r=1.4}))
  n7 = MaterialFunctionCall!(Diffuse 1, Diffuse 2, Transistion Phase=@n6, Height Texture 1=@n5, Height Texture 2=@n4.Height, Contrast=Constant() {r=0.5}) {materialFunction=/Engine/Functions/Engine_MaterialFunctions02/Texturing/HeightLerpWithTwoHeightMaps}
  FunctionOutput(None=BlendMaterialAttributes(A=@n4.Material, B=FunctionInput() {inputName=Material, inputType=FunctionInput_MaterialAttributes, sortPriority=-2}, Alpha=@n7.Alpha)) {outputName=Material}
  n8 = Add(A=@n5, B=Constant() {r=-1})
  FunctionOutput(None=LinearInterpolate(A=@n4.Height, B=Saturate(None=@n8), Alpha=@n7.Alpha)) {outputName=HeightmapForSnow, sortPriority=2}
  n9 = Saturate(None=Add(A=@n3, B=Constant() {r=0.1}))
  n10 = Constant() {r=-0.42}
  n11 = Constant() {r=-0.42}
  n12 = Add(A=Saturate(None=Add(A=@n3, B=@n10)), B=Saturate(None=Add(A=@n7.Lerp Alpha No Contrast, B=@n11)))
  n13 = Saturate(None=ShadingPathSwitch(Default=@n12, Deferred=@n12, Forward=@n9, Mobile=@n9))
  FunctionOutput(None=Ceil(None=OneMinus(None=@n13))) {outputName=LandscapeGrassMeshMask, sortPriority=1}
  n14 = Power(Base=@n3, Exp=Constant() {r=3})
  FunctionOutput(None=Multiply(A=OneMinus(None=@n14), B=Constant() {r=0.3})) {outputName=MaskForGrassBorderColor, sortPriority=3}
  FunctionOutput(None=MaterialFunctionCall(Transition Phase=@n6, Height Texture=@n4.Height) {materialFunction=&HeightLerp}.Lerp Alpha No Contrast) {outputName=HeightmapForGrass, sortPriority=3}
  FunctionOutput(None=@n7.Resulting Height Map) {outputName=RegularHeightmap, sortPriority=2}
 ]}
