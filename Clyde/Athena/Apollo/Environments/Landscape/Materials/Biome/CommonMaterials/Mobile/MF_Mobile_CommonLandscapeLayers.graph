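# Mobile landscape material function for Apollo that layers the common biome surfaces into one material attributes set.
# Starting from grass, it successively height-blends gravel, mud, sand, asphalt and cliff through MF_BlendLandscapeLayers, with mud reusing tinted gravel color.
# Combined painted-layer and cliff slope masks feed the grass function and a Grass_Standard mask output.
# It also outputs a heightmap, and a UseSnow static switch optionally blends in procedural snow.
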MaterialFunction /Clyde/Athena/Apollo/Environments/Landscape/Materials/Biome/CommonMaterials/Mobile/MF_Mobile_CommonLandscapeLayers
graph {nodes [
  Constant3Vector() {constant=<rgba 0.45 0.3773 0.225 1>}
  n1 = Max(
   A=Max(A=LandscapeLayerSample() {parameterName=Mud}, B=LandscapeLayerSample() {parameterName=Gravel})
   B=LandscapeLayerSample() {parameterName=Sand}
  )
  n2 = LandscapeLayerSample() {parameterName=Asphalt}
  n3 = Saturate(None=Multiply(A=Max(A=@n1, B=@n2), B=Constant() {r=3}))
  n4 = LandscapeLayerSample() {parameterName=Gravel}
  n5 = MaterialFunctionCall!(OverrideEdgeColor, EdgeMask=@n4, BypassBiomeColors) {materialFunction=/Game/Athena/Apollo/Environments/Landscape/Materials/Biome/CommonMaterials/Mobile/MF_Mobile_Common_Gravel}
  n6 = MaterialFunctionCall!(Gravel=@n5.RGBA) {materialFunction=&/MF_Mobile_Common_Cliff}
  n7 = MaterialFunctionCall!(Slope-CliffMask=@n6.SoftSlope, CombinedLandscapeLayersMask=@n3) {materialFunction=&/MF_Mobile_Common_Grass}
  n8 = MaterialFunctionCall!(
   Material 1=@n7.Material
   Height Texture 1=@n7.Height
   Material 2=@n7.Material
   Height Texture 2=@n7.Height
   Transistion Phase=LandscapeLayerSample() {parameterName=NoFoliage}
   UseContrast
   Contrast
  ) {materialFunction=/Clyde/Athena/Apollo/Environments/Landscape/Materials/Biome/CommonMaterials/MF_BlendLandscapeLayers}
  n9 = MaterialFunctionCall(
   Material 1=@n8.BlendedMaterial
   Height Texture 1=@n8.BlendedMaterial
   Material 2=@n5.Material
   Height Texture 2=@n5.Material
   Transistion Phase=@n4
  ) {materialFunction=&MF_BlendLandscapeLayers}
  n10 = SetMaterialAttributes!(
   MaterialAttributes
   BaseColor=Multiply(A=@n5.Mobile-RGB, B=Constant3Vector() {constant=<rgba 0.5 0.4596 0.375 1>})
   Specular=Constant() {r=0.3}
   Roughness=Constant() {r=0.9}
   Normal
  ) {attributeSetTypes [
    #1
    #2
    #3
    #4
   ]}
  n11 = MaterialFunctionCall(
   Material 1=@n9.BlendedMaterial
   Height Texture 1=@n9.BlendedMaterial
   Material 2=@n10
   Height Texture 2=@n5.Height
   Transistion Phase=LandscapeLayerSample() {parameterName=Mud}
  ) {materialFunction=&MF_BlendLandscapeLayers}
  n12 = MaterialFunctionCall() {materialFunction=&/MF_Mobile_Common_Sand}
  n13 = MaterialFunctionCall(
   Material 1=@n11.BlendedMaterial
   Height Texture 1=@n11.BlendedMaterial
   Material 2=@n12.Material
   Height Texture 2=@n12.Material
   Transistion Phase=LandscapeLayerSample() {parameterName=Sand}
  ) {materialFunction=&MF_BlendLandscapeLayers}
  n14 = MaterialFunctionCall!(TilingSize) {materialFunction=/Clyde/Athena/Apollo/Environments/Landscape/Materials/MaterialFunctions/MF_ApolloAsphalt}
  n15 = SetMaterialAttributes(
   BaseColor=ComponentMask(None=GetMaterialAttributes(None=@n14.Result) {attributeGetTypes [#1]}.BaseColor) {b=true, g=true, r=true}
   Specular=Constant() {r=0.3}
   Roughness=Constant() {r=0.9}
  ) {attributeSetTypes [
    #1
    #2
    #3
    #4
   ]}
  n16 = MaterialFunctionCall(
   Material 1=@n13.BlendedMaterial
   Height Texture 1=@n13.BlendedMaterial
   Material 2=@n15
   Height Texture 2=@n14.CrackHeight
   Transistion Phase=LandscapeLayerSample() {parameterName=Asphalt}
  ) {materialFunction=&MF_BlendLandscapeLayers}
  n17 = MaterialFunctionCall(
   Material 1=@n16.BlendedMaterial
   Height Texture 1=@n16.BlendedMaterial
   Material 2=@n6.Material
   Height Texture 2=@n6.Material
   Transistion Phase=@n6.Material
  ) {materialFunction=&MF_BlendLandscapeLayers}
  n18 = MaterialFunctionCall!(MaterialIn=@n17.BlendedMaterial) {materialFunction=/Clyde/Athena/Apollo/Environments/Landscape/Materials/Biome/UniqueBiomeMaterialFunctions/MF_ProceduralSnow}
  n19 = StaticBool()
  FunctionOutput(None=StaticSwitch(
    True=BlendMaterialAttributes(A=@n17.BlendedMaterial, B=@n18.MobileMaterial, Alpha=@n18.MobileMaterial) {desc="Vertex blend type altered!", vertexAttributeBlendType=UseA}
    False=@n17.BlendedMaterial
    Value=FunctionInput(Preview=@n19) {bUsePreviewValueAsDefault=true, inputName=UseSnow, inputType=FunctionInput_StaticBool}
   )) {outputName=Material, sortPriority=-1}
  FunctionOutput(None=@n17.Resulting Height Map) {outputName=Heightmap}
  FunctionOutput(None=Subtract(A=OneMinus(None=@n3), B=Ceil(None=@n6.Slope))) {outputName=Mask-Grass_Standard}
  Reroute(None=@n6.SoftSlope)
  Saturate(None=Power(Base=@n6.SoftSlope, Exp=Constant() {r=10}))
 ]}
