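# Material function for the farm grass landscape biome that recolors an incoming grass material's base color.
# Grasslands and farm colors blend by the terrain Farm mask, then toward a third color by runtime virtual texture output level when virtual texturing is on.
# The noise input blends a first color into that result, which the detail input modulates and slightly lifts.
# Height output passes the noise mask through; the GrassHeightmap input is unused.
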MaterialFunction /Clyde/Athena/Apollo/Environments/Landscape/Materials/Biome/UniqueBiomeMaterialFunctions/MF_Biome_FarmGrass
graph {nodes [
  Constant() {r=0.55}
  Constant() {r=3}
  FunctionInput() {inputName=GrassHeightmap, inputType=FunctionInput_Scalar}
  n1 = FunctionInput() {inputName=GrassBaseColorDetail, inputType=FunctionInput_Scalar}
  n2 = Multiply(A=Constant() {r=0.05}, B=@n1)
  n3 = Constant() {r=0.55}
  n4 = Constant() {r=3}
  n5 = LinearInterpolate(
   A=VectorParameter() {parameterName=GrasslandsGrassColorB, defaultValue=<rgba 0.35 0.1986 0.035 1>}.RGB
   B=VectorParameter() {parameterName=FarmGrassColorB, defaultValue=<rgba 0.0354 0.1305 0.5 1>}.RGB
   Alpha=MaterialFunctionCall!(UseLandscapeCoords) {materialFunction=/Clyde/Athena/Apollo/Environments/Landscape/Materials/Biome/CommonMaterials/MF_ApolloTerrainMasks}.Farm
  )
  n6 = VectorParameter() {parameterName=FarmGrassColorC, defaultValue=<rgba 0.1134 0.18 0.0181 1>}
  n7 = ViewProperty() {property=MEVP_RuntimeVirtualTextureOutputLevel}
  n8 = ScalarParameter() {defaultValue=1.5, parameterName=RVT_DistNearFarmGrass}
  n9 = VirtualTextureFeatureSwitch!(No=Constant(), Yes=Saturate(None=Multiply(A=@n7.InvProperty, B=@n8)))
  n10 = FunctionInput() {inputName=GrassNoiseMask, inputType=FunctionInput_Scalar}
  n11 = Multiply(
   A=Saturate(None=LinearInterpolate(A=@n3, B=@n4, Alpha=@n1))
   B=LinearInterpolate(
    A=VectorParameter() {parameterName=FarmGrassColorA, defaultValue=<rgba 0.0945 0.15 0.015 1>}.RGB
    B=LinearInterpolate(A=@n5, B=@n6.RGB, Alpha=@n9)
    Alpha=@n10
   )
  )
  FunctionOutput(None=SetMaterialAttributes!(MaterialAttributes=FunctionInput() {inputName=GrassMaterial, inputType=FunctionInput_MaterialAttributes}, BaseColor=Add(A=@n2, B=@n11)) {attributeSetTypes [#1]}) {outputName=Material}
  FunctionOutput(None=@n10) {outputName=Height, sortPriority=1}
 ]}
