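# Landscape biome material function for the Apollo forest floor, outputting material attributes and a height value.
# Forest floor diffuse, mask and normal textures share landscape layer UVs and shared wrap samplers.
# Diffuse color passes through MF_LS_DivideColor with the biome color, divisor, mask and mask-toggle inputs, then is saturated into BaseColor.
# The mask texture's red channel drives both Specular and Roughness, and diffuse alpha supplies Height.
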MaterialFunction /Clyde/Athena/Apollo/Environments/Landscape/Materials/Biome/UniqueBiomeMaterialFunctions/MF_Biome_Forest
graph {nodes [
  n1 = FunctionInput(Preview=StaticBool()) {bUsePreviewValueAsDefault=true, inputName=UseBiomeColorMask, inputType=FunctionInput_StaticBool, previewValue=<64 64 64 6>}
  n2 = FunctionInput(Preview=Constant() {r=1}) {bUsePreviewValueAsDefault=true, inputName=BiomeColorMask, inputType=FunctionInput_Scalar, sortPriority=1}
  n3 = FunctionInput() {bUsePreviewValueAsDefault=true, inputName=BiomeColorDiv, sortPriority=2, previewValue=<1 1 1 1>}
  n4 = FunctionInput() {bUsePreviewValueAsDefault=true, inputName=BiomeColor, sortPriority=3, previewValue=<1 1 1 1>}
  n5 = LandscapeLayerCoords() {mappingScale=10}
  n6 = TextureSample(UVs=@n5) {samplerSource=SSM_Wrap_WorldGroupSettings, texture=/Game/Environments/Apollo/Sets/Terrain/Textures/T_Apollo_Forest_Floor_D_01}
  n7 = TextureSample(UVs=@n5) {
   samplerSource=SSM_Wrap_WorldGroupSettings
   samplerType=SAMPLERTYPE_Masks
   texture=/Game/Environments/Apollo/Sets/Terrain/Textures/T_Apollo_Forest_Floor_S_01
  }
  n8 = TextureSample(UVs=@n5) {
   samplerSource=SSM_Wrap_WorldGroupSettings
   samplerType=SAMPLERTYPE_Normal
   texture=/Game/Environments/Apollo/Sets/Terrain/Textures/T_Apollo_Forest_Floor_N_01
  }
  n9 = SetMaterialAttributes!(
   MaterialAttributes
   BaseColor=Saturate(None=MaterialFunctionCall!(UseBiomeColorMask=@n1, BiomeColorMask=@n2, BiomeColorDiv=@n3, BiomeColor=@n4, Basecolor=@n6.RGB) {materialFunction=/Clyde/Athena/Apollo/Environments/Landscape/Materials/Biome/CommonMaterials/MF_LS_DivideColor}.Result)
   Specular=@n7.R
   Roughness=@n7.R
   Normal=@n8.RGB
  ) {attributeSetTypes [
    #1
    #2
    #3
    #4
   ]}
  FunctionOutput(None=@n9) {outputName=Material}
  FunctionOutput(None=@n6.A) {outputName=Height, sortPriority=1}
 ]}
