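# Material function for the Wheat biome landscape layer, outputting wheat-colored material attributes and a blend height.
# A landscape-tiled Transylvania grass texture drives color variation from green and specular and roughness from alpha, with the GrassNoiseMask input modulating color, specular and height.
# The normal is the GrassNormals input tinted, inverted and scaled down.
# GrassMaterial and GrassHeightmap inputs, two normal-map samples and three color constants are unconnected.
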MaterialFunction /Clyde/Athena/Apollo/Environments/Landscape/Materials/Biome/UniqueBiomeMaterialFunctions/MF_Biome_Wheat
graph {nodes [
  Constant3Vector() {constant=<rgba 0.3 0.1684 0.09 1>}
  Constant3Vector() {constant=<rgba 1.6 1.1102 0.32 1>}
  Constant3Vector() {constant=<rgba 0.75 0.6521 0.45 1>}
  FunctionInput() {inputName=GrassMaterial, inputType=FunctionInput_MaterialAttributes}
  FunctionInput() {inputName=GrassHeightmap, inputType=FunctionInput_Scalar}
  n1 = Multiply(A=FunctionInput() {inputName=GrassNoiseMask, inputType=FunctionInput_Scalar})
  n2 = Add(A=@n1, B=ScalarParameter() {defaultValue=0.5, parameterName=gdrty57rh54422})
  n3 = Constant() {r=0.1}
  n4 = Constant() {r=0.5}
  n5 = LandscapeLayerCoords() {mappingScale=7}
  n6 = TextureSample(UVs=@n5) {texture=/Game/Environments/Sets/Transylvania/Terrain/T_TRV_Grass_01_D}
  n7 = Multiply(A=FunctionInput() {inputName=GrassNormals}, B=VectorParameter() {parameterName=WheatNormalMult, defaultValue=<rgba 2 1.567 0.0991 1>}.RGB)
  n8 = LinearInterpolate(
   A=VectorParameter() {parameterName=Wheat-ColorA, defaultValue=<rgba 0.5833 0.4244 0.0891 1>}.RGB
   B=VectorParameter() {parameterName=Wheat-ColorB, defaultValue=<rgba 0.6597 0.5275 0.1495 1>}.RGB
   Alpha=@n6.G
  ) {constAlpha=0}
  n9 = VectorParameter() {parameterName=Wheat-ColorC, defaultValue=<rgba 0.6771 0.3923 0.0729 1>}
  n10 = SetMaterialAttributes!(
   MaterialAttributes
   BaseColor=Multiply(A=LinearInterpolate(A=@n9.RGB, Alpha=@n1), B=Saturate(None=@n8))
   Specular=Multiply(A=Saturate(None=@n2), B=LinearInterpolate(A=@n3, B=@n4, Alpha=@n6.A))
   Roughness=LinearInterpolate(A=Constant() {r=1}, B=Constant() {r=0.8}, Alpha=@n6.A)
   Normal=Multiply(A=OneMinus(None=@n7)) {constB=0.1}
  ) {attributeSetTypes [
    #1
    #2
    #3
    #4
   ]}
  FunctionOutput(None=@n10) {outputName=Material}
  FunctionOutput(None=Multiply(A=@n1, B=@n6.A)) {outputName=Height, sortPriority=1}
  TextureSample(UVs=@n5) {samplerType=SAMPLERTYPE_Normal, texture=/Game/Environments/Sets/Transylvania/Terrain/T_TRV_Grass_01_N}
  TextureSample(UVs=LandscapeLayerCoords() {mappingScale=50}) {samplerType=SAMPLERTYPE_Normal, texture=/Game/Packages/DS_Fortnite_Terrain_NoLOD/Textures/T_Grasslands_N}
 ]}
