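# Material function in the Apollo landscape virtual texture folder that converts a material's normal from tangent space to world space.
# It takes a Material Attributes input, extracts the Normal and runs it through a vector Transform.
# The result is set back into the same attributes, so every other attribute passes through unchanged.
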MaterialFunction /Clyde/Athena/Apollo/Maps/Landscape/VT/MaterialFunctions/TransformTangentToWorld
graph {nodes [
  n1 = FunctionInput() {inputName=Material, inputType=FunctionInput_MaterialAttributes}
  n2 = GetMaterialAttributes(None=@n1) {attributeGetTypes [#1]}
  FunctionOutput(None=SetMaterialAttributes!(MaterialAttributes=@n1, Normal=Transform(None=@n2.Normal)) {attributeSetTypes [#1]})
 ]}
