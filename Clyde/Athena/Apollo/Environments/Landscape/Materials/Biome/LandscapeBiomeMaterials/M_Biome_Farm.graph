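# Farm biome landscape material that shades terrain and drives procedural grass, bush and flower spawning from paint-layer masks.
# Height-blends common ground layers with wheat, forest and farm-grass biome functions via BiomeA/B/C weights, then applies slope cliffs and procedural snow.
# Writes the result to a runtime virtual texture and can sample it back through a static switch.
# A simplified mobile layer function covers non-virtual-texture and mobile quality paths.
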Material /Clyde/Athena/Apollo/Environments/Landscape/Materials/Biome/LandscapeBiomeMaterials/M_Biome_Farm
graph {
 properties {
  bTangentSpaceNormal=false
  bUseMaterialAttributes=true
  bUsedWithNanite=true
  bUsedWithVirtualHeightfieldMesh=true
  pixelDepthOffsetMode=PDOM_Legacy
  usedShadingModels=Default Lit
 }
 nodes [
  n1 = LandscapeLayerSample() {parameterName=BiomeC}
  n2 = Constant() {r=2}
  n3 = LandscapeLayerSample() {parameterName=BiomeB}
  n4 = Max(A=@n1, B=Saturate(None=Multiply(A=@n2, B=@n3)))
  n5 = LandscapeLayerSample() {parameterName=Gravel}
  n6 = LandscapeLayerSample() {parameterName=Mud}
  n7 = Max(A=Multiply(A=Max(A=@n5, B=@n6), B=Constant() {r=2}), B=Multiply(A=LandscapeLayerSample() {parameterName=Sand}, B=Constant() {r=5}))
  n8 = Ceil(None=LandscapeLayerSample() {parameterName=Asphalt})
  n9 = MaterialFunctionCall!(UseLandscapeCoords=StaticBool() {value=true}) {materialFunction=/Clyde/Athena/Apollo/Environments/Landscape/Materials/Biome/CommonMaterials/MF_ApolloTerrainMasks}
  n10 = LandscapeLayerSample() {parameterName=BiomeB}
  n11 = MaterialFunctionCall!(
   UseBiomeColorMask_1=StaticBool() {value=true}
   BiomeMask_1=@n9.Farm
   GrassBorderColorMask_1=Max(A=@n10, B=@n19.MaskForGrassBorderColor)
   HeightmapFromCliff_1=@n19.HeightmapForGrass
   Grass-EdgeBiomeColorB_1=VectorParameter() {parameterName=GrassEdgeColorA, defaultValue=<rgba 0.1458 0.1016 0 1>}.RGB
   Grass-EdgeBiomeColorA_1=VectorParameter() {parameterName=GrassEdgeColorB, defaultValue=<rgba 0.2257 0.1726 0 1>}.RGB
   GravelBiomeColorDiv_1=VectorParameter() {parameterName=Biome-GravelDiv, defaultValue=<rgba 0.8 0.7395 0.64 1>}.RGB
   GravelBiomeColor_1=VectorParameter() {parameterName=Biome-GravelColor, defaultValue=<rgba 0.2986 0.2272 0.0498 1>}.RGB
   MudBiomeColorDiv_1=VectorParameter() {parameterName=Biome-MudDiv, defaultValue=<rgba 0.7 0.647 0.56 1>}.RGB
   MudBiomeColor_1=VectorParameter() {parameterName=Biome-MudColor, defaultValue=<rgba 0.184 0.1629 0.1195 1>}.RGB
   MudBiomeRoughMod_1
   SandBiomeColorDiv_1=VectorParameter() {parameterName=Biome-SandDiv, defaultValue=<rgba 1 0.896 0.725 1>}.RGB
   SandBiomeColor_1=VectorParameter() {parameterName=Biome-SandColor, defaultValue=<rgba 0.2951 0.2519 0.0789 1>}.RGB
  ) {materialFunction=/Clyde/Athena/Apollo/Environments/Landscape/Materials/Biome/CommonMaterials/MF_CommonLandscapeLayers}
  n12 = MaterialFunctionCall!(
   Material 1=@n11.Material
   Height Texture 1=@n11.Heightmap
   Material 2=@n11.Material
   Height Texture 2=@n11.Heightmap
   Transistion Phase=LandscapeLayerSample() {parameterName=NoFoliage}
   UseContrast
   Contrast
  ) {materialFunction=/Clyde/Athena/Apollo/Environments/Landscape/Materials/Biome/CommonMaterials/MF_BlendLandscapeLayers}
  n13 = MaterialFunctionCall!(GrassHeightmap, GrassNoiseMask=@n11.GrassHeightmap_1, GrassNormals=@n11.GrassHeightmap_1, GrassMaterial) {materialFunction=/Clyde/Athena/Apollo/Environments/Landscape/Materials/Biome/UniqueBiomeMaterialFunctions/MF_Biome_Wheat}
  n14 = MaterialFunctionCall(
   Material 1=@n12.BlendedMaterial
   Height Texture 1=@n12.Resulting Height Map
   Material 2=@n13.Material
   Height Texture 2=@n13.Material
   Transistion Phase=LandscapeLayerSample() {parameterName=BiomeA}
  ) {materialFunction=&MF_BlendLandscapeLayers}
  n15 = MaterialFunctionCall!(UseBiomeColorMask, BiomeColorMask, BiomeColorDiv, BiomeColor) {materialFunction=/Clyde/Athena/Apollo/Environments/Landscape/Materials/Biome/UniqueBiomeMaterialFunctions/MF_Biome_Forest}
  n16 = MaterialFunctionCall(
   Material 1=@n14.BlendedMaterial
   Height Texture 1=@n14.BlendedMaterial
   Material 2=@n15.Material
   Height Texture 2=@n15.Material
   Transistion Phase=@n10
  ) {materialFunction=&MF_BlendLandscapeLayers}
  n17 = MaterialFunctionCall!(
   GrassHeightmap=@n11.GrassNormal_1
   GrassNoiseMask=@n11.GrassHeightmap_1
   GrassBaseColorDetail=@n11.GrassBaseColorDetail_1
   GrassMaterial=@n11.Material
  ) {materialFunction=/Clyde/Athena/Apollo/Environments/Landscape/Materials/Biome/UniqueBiomeMaterialFunctions/MF_Biome_FarmGrass}
  n18 = MaterialFunctionCall(
   Material 1=@n16.BlendedMaterial
   Height Texture 1=@n16.BlendedMaterial
   Material 2=@n17.Material
   Height Texture 2=@n17.Material
   Transistion Phase=LandscapeLayerSample() {parameterName=BiomeC}
  ) {materialFunction=&MF_BlendLandscapeLayers}
  n19 = MaterialFunctionCall!(
   BiomeColorMask=@n9.Farm
   Material=@n18.BlendedMaterial
   Heightmap=@n18.Resulting Height Map
   UseBiomeColorMask=StaticBool()
   BiomeColorDiv
   BiomeColor
  ) {materialFunction=/Clyde/Athena/Apollo/Environments/Landscape/Materials/Biome/CommonMaterials/MF_SlopeBlendCliff}
  n20 = Max(A=@n19.LandscapeGrassMeshMask, B=Saturate(None=Max(A=@n7, B=@n8)))
  MaterialExpressionSubtract_0 = Subtract(A=OneMinus(None=Max(A=@n4, B=@n20)), B=LandscapeLayerSample() {parameterName=NoFoliage})
  n21 = Constant() {r=2}
  MaterialExpressionSubtract_1 = Subtract(A=Add(A=Constant() {r=-0.5}, B=Multiply(A=@n21, B=@n1)), B=LandscapeLayerSample() {parameterName=NoFoliage})
  MaterialExpressionSubtract_9 = Subtract(A=Constant(), B=LandscapeLayerSample() {parameterName=NoFoliage})
  LandscapeGrassOutput!(
   Grass_Standard=@MaterialExpressionSubtract_0
   Bush=@MaterialExpressionSubtract_0
   FarmFlowers=@MaterialExpressionSubtract_1
   Scree=@MaterialExpressionSubtract_9
   Scree_Grass=@MaterialExpressionSubtract_9
   Heather=@MaterialExpressionSubtract_9
   SwampGrass=@MaterialExpressionSubtract_9
   SwampReeds=@MaterialExpressionSubtract_9
   Forest_Grass=@MaterialExpressionSubtract_9
   Beach=@MaterialExpressionSubtract_9
   Ocean=@MaterialExpressionSubtract_9
  ) {grassTypes [
    Grass_Standard {
     grassType=/Clyde/Landscape/GrassTypes/LandscapeGrassType_StandardGrass
     input {
      expression=&self:MaterialExpressionSubtract_0
      inputName=None
      mask=0
      maskA=0
      maskB=0
      maskG=0
      maskR=0
      outputIndex=0
     }
    }
    Bush {
     grassType=/Clyde/Landscape/GrassTypes/Athena_Foliage_Bushes
     input {
      expression=&self:MaterialExpressionSubtract_0
      inputName=None
      mask=0
      maskA=0
      maskB=0
      maskG=0
      maskR=0
      outputIndex=0
     }
    }
    FarmFlowers {
     grassType=/Clyde/Landscape/GrassTypes/LandscapeGrassType_FarmFlowers
     input {
      expression=&self:MaterialExpressionSubtract_1
      inputName=None
      mask=0
      maskA=0
      maskB=0
      maskG=0
      maskR=0
      outputIndex=0
     }
    }
    Scree {
     grassType=/Clyde/Landscape/GrassTypes/LandscapeGrassType_Scree
     input {
      expression=&self:MaterialExpressionSubtract_9
      inputName=None
      mask=0
      maskA=0
      maskB=0
      maskG=0
      maskR=0
      outputIndex=0
     }
    }
    Scree_Grass {
     grassType=/Clyde/Landscape/GrassTypes/LandscapeGrassType_ScreeGrass
     input {
      expression=&self:MaterialExpressionSubtract_9
      inputName=None
      mask=0
      maskA=0
      maskB=0
      maskG=0
      maskR=0
      outputIndex=0
     }
    }
    Heather {
     grassType=/Clyde/Landscape/GrassTypes/LandscapeGrassType_Heather
     input {
      expression=&self:MaterialExpressionSubtract_9
      inputName=None
      mask=0
      maskA=0
      maskB=0
      maskG=0
      maskR=0
      outputIndex=0
     }
    }
    SwampGrass {
     grassType=/Clyde/Landscape/GrassTypes/LandscapeGrassType_SwampGrass
     input {
      expression=&self:MaterialExpressionSubtract_9
      inputName=None
      mask=0
      maskA=0
      maskB=0
      maskG=0
      maskR=0
      outputIndex=0
     }
    }
    SwampReeds {
     grassType=/Clyde/Landscape/GrassTypes/LandscapeGrassType_SwampReeds
     input {
      expression=&self:MaterialExpressionSubtract_9
      inputName=None
      mask=0
      maskA=0
      maskB=0
      maskG=0
      maskR=0
      outputIndex=0
     }
    }
    Forest_Grass {
     grassType=/Clyde/Landscape/GrassTypes/LandscapeGrassType_ForestGrass
     input {
      expression=&self:MaterialExpressionSubtract_9
      inputName=None
      mask=0
      maskA=0
      maskB=0
      maskG=0
      maskR=0
      outputIndex=0
     }
    }
    Beach {
     grassType=/Clyde/Landscape/GrassTypes/LandscapeGrassType_Beach
     input {
      expression=&self:MaterialExpressionSubtract_9
      inputName=None
      mask=0
      maskA=0
      maskB=0
      maskG=0
      maskR=0
      outputIndex=0
     }
    }
    Ocean {
     grassType=/Clyde/Landscape/GrassTypes/LandscapeGrassType_Ocean
     input {
      expression=&self:MaterialExpressionSubtract_9
      inputName=None
      mask=0
      maskA=0
      maskB=0
      maskG=0
      maskR=0
      outputIndex=0
     }
    }
   ]}
  n22 = GetMaterialAttributes(None=MaterialFunctionCall!(UseSnow=StaticBool() {value=true}) {materialFunction=/Clyde/Athena/Apollo/Environments/Landscape/Materials/Biome/CommonMaterials/Mobile/MF_Mobile_CommonLandscapeLayers}.Material) {attributeGetTypes [#1]}
  MaterialExpressionConstant_13 = Constant()
  n23 = LandscapeLayerBlend!(
   Layer Gravel=@MaterialExpressionConstant_13
   Layer Mud=@MaterialExpressionConstant_13
   Layer Sand=@MaterialExpressionConstant_13
   Layer Asphalt=@MaterialExpressionConstant_13
   Layer NoFoliage=@MaterialExpressionConstant_13
  ) {layers [
    {
     blendType=LB_WeightBlend
     constHeightInput=0
     layerName=Gravel
     previewWeight=0
     constLayerInput=<0 0 0>
     heightInput {expression=null, inputName=None, mask=0, maskA=0, maskB=0, maskG=0, maskR=0, outputIndex=0}
     layerInput {
      expression=&self:MaterialExpressionConstant_13
      inputName=None
      mask=0
      maskA=0
      maskB=0
      maskG=0
      maskR=0
      outputIndex=0
     }
    }
    {
     blendType=LB_WeightBlend
     constHeightInput=0
     layerName=Mud
     previewWeight=0
     constLayerInput=<0 0 0>
     heightInput {expression=null, inputName=None, mask=0, maskA=0, maskB=0, maskG=0, maskR=0, outputIndex=0}
     layerInput {
      expression=&self:MaterialExpressionConstant_13
      inputName=None
      mask=0
      maskA=0
      maskB=0
      maskG=0
      maskR=0
      outputIndex=0
     }
    }
    {
     blendType=LB_WeightBlend
     constHeightInput=0
     layerName=Sand
     previewWeight=0
     constLayerInput=<0 0 0>
     heightInput {expression=null, inputName=None, mask=0, maskA=0, maskB=0, maskG=0, maskR=0, outputIndex=0}
     layerInput {
      expression=&self:MaterialExpressionConstant_13
      inputName=None
      mask=0
      maskA=0
      maskB=0
      maskG=0
      maskR=0
      outputIndex=0
     }
    }
    {
     blendType=LB_WeightBlend
     constHeightInput=0
     layerName=Asphalt
     previewWeight=0
     constLayerInput=<0 0 0>
     heightInput {expression=null, inputName=None, mask=0, maskA=0, maskB=0, maskG=0, maskR=0, outputIndex=0}
     layerInput {
      expression=&self:MaterialExpressionConstant_13
      inputName=None
      mask=0
      maskA=0
      maskB=0
      maskG=0
      maskR=0
      outputIndex=0
     }
    }
    {
     blendType=LB_WeightBlend
     constHeightInput=0
     layerName=NoFoliage
     previewWeight=0
     constLayerInput=<0 0 0>
     heightInput {expression=null, inputName=None, mask=0, maskA=0, maskB=0, maskG=0, maskR=0, outputIndex=0}
     layerInput {
      expression=&self:MaterialExpressionConstant_13
      inputName=None
      mask=0
      maskA=0
      maskB=0
      maskG=0
      maskR=0
      outputIndex=0
     }
    }
   ]}
  n24 = SetMaterialAttributes!(MaterialAttributes=@n22.MaterialAttributes, Base Color=Add(A=@n22.Base Color, B=Multiply(A=@n23) {constB=0})) {attributeSetTypes [#1]}
  n25 = RuntimeVirtualTextureSampleParameter() {
   group=RVT
   materialType=BaseColor_Normal_Specular_YCoCg
   parameterName=RVT Sample Target
   virtualTexture=/Game/Athena/Apollo/RVT/RVT_Apollo_Landscape
  }
  n26 = SetMaterialAttributes!(MaterialAttributes, Base Color=@n25.BaseColor, Specular=@n25.BaseColor, Roughness=@n25.BaseColor, Normal=@n25.BaseColor) {attributeSetTypes [
    #1
    #2
    #3
    #4
   ]}
  n27 = MaterialFunctionCall!(Material=@n19.Material) {materialFunction=/Clyde/Athena/Apollo/Maps/Landscape/VT/MaterialFunctions/TransformTangentToWorld}
  n28 = MaterialFunctionCall!(MaterialIn=@n27.Result) {materialFunction=/Clyde/Athena/Apollo/Environments/Landscape/Materials/Biome/UniqueBiomeMaterialFunctions/MF_ProceduralSnow}
  n29 = BlendMaterialAttributes(A=@n27.Result, B=@n28.Material, Alpha=@n28.Material) {desc="Vertex blend type altered!", vertexAttributeBlendType=UseA}
  n30 = VirtualTextureFeatureSwitch!(
   No=@n24
   Yes=StaticSwitchParameter(
    True=MaterialFunctionCall!(RVT=@n26, ByPass RVT=@n29) {materialFunction=/Game/Athena/Apollo/Maps/Landscape/VT/MaterialFunctions/EnableRVT}.Result
    False=@n29
   ) {defaultValue=true, group=RVT Base, parameterName=RVT Enable}
  )
  MaterialFunctionCall!(Material=ShadingPathSwitch(Default=@n30, Mobile=QualitySwitch(Default=@n24, Medium=@n30)), DebugAsBasecolor, DebugBiomeColor) {materialFunction=/Clyde/Athena/Apollo/Environments/Landscape/Materials/Biome/CommonMaterials/MF_LS_DEBUG}
  n31 = GetMaterialAttributes(None=@n29) {attributeGetTypes [
    #1
    #2
    #3
    #4
   ]}
  RuntimeVirtualTextureOutput!(
   BaseColor=MaterialFunctionCall!(BaseColor=@n31.Base Color) {materialFunction=/Game/Athena/Apollo/Environments/Landscape/Materials/Biome/CommonMaterials/MF_ReduceRTV-CompressionArtifacts}.BaseColor
   Specular=@n31.Specular
   Roughness=@n31.Specular
   Normal=@n31.Specular
   WorldHeight=Add(A=BreakMaterialAttributes!(Attr=@n31.MaterialAttributes).Displacement, B=WorldPosition().Z)
   Opacity
   Mask
   Displacement
   Mask4
  )
  StaticBool()
  VectorParameter() {parameterName=DebugBiomeColor, defaultValue=<rgba 0.0445 1 0 1>}
  VectorParameter() {parameterName=Biome-CliffDiv, defaultValue=<rgba 0.5375 0.5588 0.75 1>}
  VectorParameter() {parameterName=Biome-CliffColor, defaultValue=<rgba 0.8 0.8 0.8 1>}
 ]
}
