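# Landscape material function that, per its name, reduces runtime virtual texture compression artifacts by adding faint noise to BaseColor.
# It samples the T_Apollo_TCNoise texture as linear color, using scaled world-position XY as UVs for a top-down planar projection.
# The noise RGB is multiplied by a tiny negative constant and added to the incoming BaseColor, slightly darkening it in a dither-like variation.
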MaterialFunction /Clyde/Athena/Apollo/Environments/Landscape/Materials/Biome/CommonMaterials/MF_ReduceRTV-CompressionArtifacts
graph {nodes [
  Constant() {r=0.004}
  Constant() {r=-0.005}
  n1 = WorldPosition()
  n2 = Constant() {r=0.004}
  n3 = TextureSample(UVs=ComponentMask(None=Multiply(A=@n1.XYZ, B=@n2)) {g=true, r=true}) {samplerType=SAMPLERTYPE_LinearColor, texture=&/T_Apollo_TCNoise}
  n4 = Constant() {r=-0.005}
  FunctionOutput(None=Add(A=Multiply(A=@n3.RGB, B=@n4) {constB=0.02}, B=FunctionInput() {inputName=BaseColor})) {outputName=BaseColor}
 ]}
