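# Material function that defines the swamp grass biome layer for the Apollo landscape, using the swamp floor texture set.
# Landscape layer coordinates tile the diffuse, mask and normal samples, which feed a SetMaterialAttributes node output as Material.
# The mask's red channel drives both Specular and Roughness.
# Diffuse alpha is exposed as a separate Height output, and all samples use shared world-group wrap samplers.
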MaterialFunction /Clyde/Athena/Apollo/Environments/Landscape/Materials/Biome/UniqueBiomeMaterialFunctions/MF_Biome_SwampGrass
graph {nodes [
  n1 = LandscapeLayerCoords() {mappingScale=5}
  n2 = TextureSample(UVs=@n1) {samplerSource=SSM_Wrap_WorldGroupSettings, texture=/Game/Environments/Apollo/Sets/Terrain/Textures/T_Apollo_Swamp_Floor_D_01}
  n3 = TextureSample(UVs=@n1) {
   samplerSource=SSM_Wrap_WorldGroupSettings
   samplerType=SAMPLERTYPE_Masks
   texture=/Game/Environments/Apollo/Sets/Terrain/Textures/T_Apollo_Swamp_Floor_S_01
  }
  n4 = TextureSample(UVs=@n1) {
   samplerSource=SSM_Wrap_WorldGroupSettings
   samplerType=SAMPLERTYPE_Normal
   texture=/Game/Environments/Apollo/Sets/Terrain/Textures/T_Apollo_Swamp_Floor_N_01
  }
  n5 = SetMaterialAttributes!(MaterialAttributes, BaseColor=@n2.RGB, Specular=@n3.R, Roughness=@n3.R, Normal=@n4.RGB) {attributeSetTypes [
    #1
    #2
    #3
    #4
   ]}
  FunctionOutput(None=@n5) {outputName=Material}
  FunctionOutput(None=@n2.A) {outputName=Height, sortPriority=1}
 ]}
