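# Landscape material function that builds the scree (loose rock) layer for Apollo biome terrain.
# It samples scree diffuse, mask and normal textures on scaled landscape UVs, multiplying base color by the shared MF_HeightBasedDarkenCliff result.
# The mask's red channel drives both Specular and Roughness.
# It outputs material attributes plus diffuse alpha as both Height and ConstantMask; the LayerWeight input is unconnected.
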MaterialFunction /Clyde/Athena/Apollo/Environments/Landscape/Materials/Biome/UniqueBiomeMaterialFunctions/MF_Scree
graph {nodes [
  FunctionInput() {inputName=LayerWeight, inputType=FunctionInput_Scalar}
  n1 = Multiply(A=LandscapeLayerCoords() {mappingScale=1}, B=Constant() {r=0.08})
  n2 = TextureSample(UVs=@n1) {samplerSource=SSM_Wrap_WorldGroupSettings, texture=/Game/Environments/Apollo/Sets/Terrain/Textures/T_Apollo_Scree_D_01}
  n3 = TextureSample(UVs=@n1) {
   samplerSource=SSM_Wrap_WorldGroupSettings
   samplerType=SAMPLERTYPE_Masks
   texture=/Game/Environments/Apollo/Sets/Terrain/Textures/T_Apollo_Scree_S_01
  }
  n4 = TextureSample(UVs=@n1) {
   samplerSource=SSM_Wrap_WorldGroupSettings
   samplerType=SAMPLERTYPE_Normal
   texture=/Game/Environments/Apollo/Sets/Terrain/Textures/T_Apollo_Scree_N_01
  }
  n5 = SetMaterialAttributes!(
   MaterialAttributes
   BaseColor=Multiply(
    A=MaterialFunctionCall!(DarknessAmount) {materialFunction=/Clyde/Athena/Apollo/Environments/Landscape/Materials/Biome/CommonMaterials/MF_HeightBasedDarkenCliff}.Result
    B=@n2.RGB
   )
   Specular=@n3.R
   Roughness=@n3.R
   Normal=@n4.RGB
  ) {attributeSetTypes [
    #1
    #2
    #3
    #4
   ]}
  FunctionOutput(None=@n5) {outputName=Material}
  FunctionOutput(None=@n2.A) {outputName=Height, sortPriority=1}
  FunctionOutput(None=@n2.A) {outputName=ConstantMask, sortPriority=1}
 ]}
