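# Nanite-enabled Apollo landscape material that blends Mountain and Swamp biomes using a terrain mask.
# Each biome height-blends painted layers (scree, forest and heather, or wet mud, forest and swamp grass), then adds slope cliffs; Mountain also gets procedural snow.
# The blend writes to a runtime virtual texture, optionally sampled back; forward and mobile paths use simplified layers.
# Layer, biome, slope and height masks drive landscape grass spawning.
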Material /Clyde/Athena/Apollo/Environments/Landscape/Materials/Biome/LandscapeBiomeMaterials/M_Biome_Mountains_SwampBlend
graph {
 properties {
  bTangentSpaceNormal=false
  bUseMaterialAttributes=true
  bUsedWithNanite=true
  physMaterial=/Game/FortressPhysicalMaterials/Grass
  pixelDepthOffsetMode=PDOM_Legacy
  usedShadingModels=Default Lit
 }
 nodes [
  Constant() {r=3909}
  Constant() {r=3909}
  n1 = Max(A=LandscapeLayerSample() {parameterName=BiomeA}, B=LandscapeLayerSample() {parameterName=BiomeB})
  n2 = LandscapeLayerSample() {parameterName=BiomeC}
  n3 = Saturate(None=Multiply(A=Constant() {r=3}, B=@n2))
  n4 = Constant() {r=2}
  n5 = LandscapeLayerSample() {parameterName=Mud}
  n6 = Max(A=Max(A=LandscapeLayerSample() {parameterName=Gravel}, B=LandscapeLayerSample() {parameterName=Sand}), B=@n5)
  n7 = LandscapeLayerSample() {parameterName=Asphalt}
  n8 = Saturate(None=Max(A=Multiply(A=@n4, B=@n6), B=Ceil(None=@n7)))
  n9 = MaterialFunctionCall!(UseLandscapeCoords=StaticBool()) {materialFunction=/Clyde/Athena/Apollo/Environments/Landscape/Materials/Biome/CommonMaterials/MF_ApolloTerrainMasks}
  n10 = LandscapeLayerSample() {parameterName=BiomeB}
  n11 = MaterialFunctionCall!(
   UseBiomeColorMask_1=StaticBool() {value=true}
   BiomeMask_1=@n9.Mountain
   GrassBorderColorMask_1=Max(A=@n10, B=@n20.MaskForGrassBorderColor)
   HeightmapFromCliff_1=@n20.HeightmapForGrass
   Grass-EdgeBiomeColorB_1=VectorParameter() {parameterName=GrassEdgeColorB, defaultValue=<rgba 0.135 0.14 0.007 1>}.RGB
   Grass-EdgeBiomeColorA_1=VectorParameter() {parameterName=GrassEdgeColorA, defaultValue=<rgba 0.0701 0.1 0.01 1>}.RGB
   GravelBiomeColorDiv_1=VectorParameter() {parameterName=Biome-GravelDiv, defaultValue=<rgba 1 0.7337 0.5061 1>}.RGB
   GravelBiomeColor_1=VectorParameter() {parameterName=Biome-GravelColor, defaultValue=<rgba 0.4275 0.4425 0.45 1>}.RGB
   MudBiomeColorDiv_1=VectorParameter() {parameterName=MountainBiomeMudDiv, defaultValue=<rgba 0.8 0.5412 0.32 1>}.RGB
   MudBiomeColor_1=VectorParameter() {parameterName=MountainBiomeMudColor, defaultValue=<rgba 0.35 0.4502 0.5 1>}.RGB
   MudBiomeRoughMod_1
   SandBiomeColorDiv_1=VectorParameter() {parameterName=BiomeSandDiv, defaultValue=<rgba 1 0.8306 0.6641 1>}.RGB
   SandBiomeColor_1=VectorParameter() {parameterName=BiomeSandColor, defaultValue=<rgba 0.2625 0.3354 0.35 1>}.RGB
  ) {materialFunction=/Clyde/Athena/Apollo/Environments/Landscape/Materials/Biome/CommonMaterials/MF_CommonLandscapeLayers}
  n12 = MaterialFunctionCall!(
   Material 1=@n11.Material
   Height Texture 1=@n11.Heightmap
   Material 2=@n11.Material
   Height Texture 2=@n11.Heightmap
   Transistion Phase=LandscapeLayerSample() {parameterName=NoFoliage}
   UseContrast
   Contrast
  ) {materialFunction=/Clyde/Athena/Apollo/Environments/Landscape/Materials/Biome/CommonMaterials/MF_BlendLandscapeLayers}
  n13 = LandscapeLayerSample() {parameterName=BiomeA}
  n14 = MaterialFunctionCall!(LayerWeight=@n13) {materialFunction=/Clyde/Athena/Apollo/Environments/Landscape/Materials/Biome/UniqueBiomeMaterialFunctions/MF_Scree}
  n15 = MaterialFunctionCall(
   Material 1=@n12.BlendedMaterial
   Height Texture 1=@n12.BlendedMaterial
   Material 2=@n14.Material
   Height Texture 2=@n14.Material
   Transistion Phase=@n13
   UseContrast=StaticBool() {value=true}
   Contrast=ScalarParameter() {defaultValue=0.5, parameterName=BiomeA-LayerContrast}
  ) {materialFunction=&MF_BlendLandscapeLayers}
  n16 = MaterialFunctionCall!(UseBiomeColorMask, BiomeColorMask, BiomeColorDiv, BiomeColor) {materialFunction=/Clyde/Athena/Apollo/Environments/Landscape/Materials/Biome/UniqueBiomeMaterialFunctions/MF_Biome_Forest}
  n17 = MaterialFunctionCall(
   Material 1=@n15.BlendedMaterial
   Height Texture 1=@n15.BlendedMaterial
   Material 2=@n16.Material
   Height Texture 2=@n16.Material
   Transistion Phase=@n10
  ) {materialFunction=&MF_BlendLandscapeLayers}
  n18 = MaterialFunctionCall!(GrassHeightmap=@n11.GrassHeightmap_1, GrassNoiseMask=@n11.GrassNoiseMask_1, GrassMaterial=@n11.Material) {materialFunction=/Clyde/Athena/Apollo/Environments/Landscape/Materials/Biome/UniqueBiomeMaterialFunctions/MF_Biome_Heather}
  n19 = MaterialFunctionCall(
   Material 1=@n17.BlendedMaterial
   Height Texture 1=@n17.BlendedMaterial
   Material 2=@n18.Material
   Height Texture 2=@n18.Material
   Transistion Phase=LandscapeLayerSample() {parameterName=BiomeC}
  ) {materialFunction=&MF_BlendLandscapeLayers}
  n20 = MaterialFunctionCall!(
   BiomeColorMask=@n9.Mountain
   Material=@n19.BlendedMaterial
   Heightmap=@n19.BlendedMaterial
   UseBiomeColorMask=StaticBool()
   BiomeColorDiv
   BiomeColor
  ) {materialFunction=/Clyde/Athena/Apollo/Environments/Landscape/Materials/Biome/CommonMaterials/MF_SlopeBlendCliff}
  n21 = OneMinus(None=Max(A=Max(A=@n1, B=@n3), B=Max(A=@n20.LandscapeGrassMeshMask, B=@n8)))
  n22 = MaterialFunctionCall(UseLandscapeCoords=StaticBool()) {materialFunction=&MF_ApolloTerrainMasks}
  n23 = Subtract(A=@n21, B=@n22.Forest) {desc=Grass_Standard}
  n24 = ScalarParameter() {defaultValue=5, parameterName=Mtn-ScreeplusForest-mod}
  n25 = LandscapeLayerSample() {parameterName=BiomeA}
  n26 = Max(A=@n25, B=LandscapeLayerSample() {parameterName=BiomeB})
  n27 = LandscapeLayerSample() {parameterName=BiomeC}
  n28 = Max(A=Saturate(None=Multiply(A=@n24, B=@n26)), B=Max(A=@n27, B=LandscapeLayerSample() {parameterName=BiomeD}))
  n29 = LandscapeLayerSample() {parameterName=Gravel}
  n30 = LandscapeLayerSample() {parameterName=Sand}
  n31 = Multiply(A=Constant() {r=2}, B=Max(A=Max(A=@n29, B=@n30), B=LandscapeLayerSample() {parameterName=Mud}))
  n32 = Ceil(None=LandscapeLayerSample() {parameterName=Asphalt})
  n33 = Max(A=@n20.LandscapeGrassMeshMask, B=Saturate(None=Max(A=@n31, B=@n32)))
  n34 = MaterialFunctionCall!(Material=@n20.Material) {materialFunction=/Game/Athena/Apollo/Maps/Landscape/VT/MaterialFunctions/TransformTangentToWorld}
  n35 = MaterialFunctionCall!(MaterialIn=@n34.Result) {materialFunction=/Clyde/Athena/Apollo/Environments/Landscape/Materials/Biome/UniqueBiomeMaterialFunctions/MF_ProceduralSnow}
  n36 = Subtract(A=OneMinus(None=Max(A=@n28, B=@n33)), B=@n35.GrassMask)
  MaterialExpressionSubtract_13 = Subtract(A=LinearInterpolate(A=@n23, B=@n36, Alpha=@n22.Mountain) {desc=Grass}, B=LandscapeLayerSample() {parameterName=NoFoliage})
  n37 = Min(A=@n22.Forest, B=@n21) {desc=Forest grass}
  n38 = Max(A=@n23, B=@n37)
  MaterialExpressionSubtract_12 = Subtract(A=LinearInterpolate(A=Max(A=@n38, B=@n2), B=@n36, Alpha=@n22.Mountain) {desc=Bush}, B=LandscapeLayerSample() {parameterName=NoFoliage})
  MaterialExpressionSubtract_5 = Subtract(A=Constant(), B=LandscapeLayerSample() {parameterName=NoFoliage})
  n39 = Constant()
  n40 = Constant() {r=2}
  n41 = Constant() {r=-0.5}
  n42 = Saturate(None=Add(A=Multiply(A=@n40, B=@n25), B=@n41))
  n43 = ScalarParameter() {defaultValue=-4, parameterName=Threshold_Scree}
  n44 = ScalarParameter() {defaultValue=1.6832, parameterName=Threshold_Scree2}
  n45 = ComponentMask(None=VertexNormalWS()) {b=true}
  n46 = Multiply(A=@n42, B=Saturate(None=LinearInterpolate(A=@n43, B=@n44, Alpha=@n45)))
  MaterialExpressionSubtract_11 = Subtract(
   A=LinearInterpolate(B=ShadingPathSwitch(Default=@n46, Deferred=@n46, Forward=@n39, Mobile=@n39), Alpha=@n22.Mountain) {desc=Scree}
   B=LandscapeLayerSample() {parameterName=NoFoliage}
  )
  n47 = ScalarParameter() {defaultValue=-4, parameterName=Threshold_ScreeGrass}
  n48 = ScalarParameter() {defaultValue=1, parameterName=Threshold_ScreeGrass2}
  n49 = ComponentMask(None=VertexNormalWS()) {b=true}
  n50 = Multiply(A=Saturate(None=LinearInterpolate(A=@n47, B=@n48, Alpha=@n49)), B=@n42)
  MaterialExpressionSubtract_10 = Subtract(
   A=LinearInterpolate(B=Subtract(A=@n50, B=@n35.GrassMask), Alpha=@n22.Mountain) {desc=Scree Grass}
   B=LandscapeLayerSample() {parameterName=NoFoliage}
  )
  n51 = Saturate(None=Add(A=@n41, B=Multiply(A=@n40, B=@n27)))
  n52 = OneMinus(None=@n20.LandscapeGrassMeshMask)
  MaterialExpressionSubtract_9 = Subtract(A=LinearInterpolate(B=Multiply(A=@n51, B=@n52), Alpha=@n22.Mountain) {desc=Heather}, B=LandscapeLayerSample() {parameterName=NoFoliage})
  MaterialExpressionSubtract_8 = Subtract(A=LinearInterpolate(A=@n2, Alpha=@n22.Mountain) {constB=0, desc=SwampGrass}, B=LandscapeLayerSample() {parameterName=NoFoliage})
  n53 = ComponentMask(None=WorldPosition().XYZ) {b=true}
  n54 = Divide(A=Add(A=@n53, B=Constant() {r=4000}), B=Constant() {r=500})
  n55 = Constant() {r=3990}
  n56 = Multiply(A=OneMinus(None=Saturate(None=@n54)), B=Saturate(None=Add(A=@n53, B=@n55)))
  MaterialExpressionSubtract_7 = Subtract(
   A=LinearInterpolate(A=Multiply(A=@n5, B=@n56), Alpha=@n22.Mountain) {constB=0, desc=SwampReed}
   B=LandscapeLayerSample() {parameterName=NoFoliage}
  )
  MaterialExpressionSubtract_6 = Subtract(A=LinearInterpolate(A=@n37, Alpha=@n22.Mountain) {constB=0, desc=Forest Grass}, B=LandscapeLayerSample() {parameterName=NoFoliage})
  MaterialExpressionSubtract_4 = Subtract(
   A=LinearInterpolate(B=Subtract(A=@n30, B=@n20.LandscapeGrassMeshMask), Alpha=@n22.Mountain) {desc=Beach}
   B=LandscapeLayerSample() {parameterName=NoFoliage}
  )
  LandscapeGrassOutput!(
   Grass_Standard=@MaterialExpressionSubtract_13
   Bush=@MaterialExpressionSubtract_12
   FarmFlowers=@MaterialExpressionSubtract_5
   Scree=@MaterialExpressionSubtract_11
   Scree_Grass=@MaterialExpressionSubtract_10
   Heather=@MaterialExpressionSubtract_9
   SwampGrass=@MaterialExpressionSubtract_8
   SwampReeds=@MaterialExpressionSubtract_7
   Forest_Grass=@MaterialExpressionSubtract_6
   Beach=@MaterialExpressionSubtract_4
   Ocean=@MaterialExpressionSubtract_5
  ) {grassTypes [
    Grass_Standard {
     grassType=/Clyde/Landscape/GrassTypes/LandscapeGrassType_StandardGrass
     input {
      expression=&self:MaterialExpressionSubtract_13
      inputName=None
      mask=0
      maskA=0
      maskB=0
      maskG=0
      maskR=0
      outputIndex=0
     }
    }
    Bush {
     grassType=/Clyde/Landscape/GrassTypes/Athena_Foliage_Bushes
     input {
      expression=&self:MaterialExpressionSubtract_12
      inputName=None
      mask=0
      maskA=0
      maskB=0
      maskG=0
      maskR=0
      outputIndex=0
     }
    }
    FarmFlowers {
     grassType=/Clyde/Landscape/GrassTypes/LandscapeGrassType_FarmFlowers
     input {
      expression=&self:MaterialExpressionSubtract_5
      inputName=None
      mask=0
      maskA=0
      maskB=0
      maskG=0
      maskR=0
      outputIndex=0
     }
    }
    Scree {
     grassType=/Clyde/Landscape/GrassTypes/LandscapeGrassType_Scree
     input {
      expression=&self:MaterialExpressionSubtract_11
      inputName=None
      mask=0
      maskA=0
      maskB=0
      maskG=0
      maskR=0
      outputIndex=0
     }
    }
    Scree_Grass {
     grassType=/Clyde/Landscape/GrassTypes/LandscapeGrassType_ScreeGrass
     input {
      expression=&self:MaterialExpressionSubtract_10
      inputName=None
      mask=0
      maskA=0
      maskB=0
      maskG=0
      maskR=0
      outputIndex=0
     }
    }
    Heather {
     grassType=/Clyde/Landscape/GrassTypes/LandscapeGrassType_Heather
     input {
      expression=&self:MaterialExpressionSubtract_9
      inputName=None
      mask=0
      maskA=0
      maskB=0
      maskG=0
      maskR=0
      outputIndex=0
     }
    }
    SwampGrass {
     grassType=/Clyde/Landscape/GrassTypes/LandscapeGrassType_SwampGrass
     input {
      expression=&self:MaterialExpressionSubtract_8
      inputName=None
      mask=0
      maskA=0
      maskB=0
      maskG=0
      maskR=0
      outputIndex=0
     }
    }
    SwampReeds {
     grassType=/Clyde/Landscape/GrassTypes/LandscapeGrassType_SwampReeds
     input {
      expression=&self:MaterialExpressionSubtract_7
      inputName=None
      mask=0
      maskA=0
      maskB=0
      maskG=0
      maskR=0
      outputIndex=0
     }
    }
    Forest_Grass {
     grassType=/Clyde/Landscape/GrassTypes/LandscapeGrassType_ForestGrass
     input {
      expression=&self:MaterialExpressionSubtract_6
      inputName=None
      mask=0
      maskA=0
      maskB=0
      maskG=0
      maskR=0
      outputIndex=0
     }
    }
    Beach {
     grassType=/Clyde/Landscape/GrassTypes/LandscapeGrassType_Beach
     input {
      expression=&self:MaterialExpressionSubtract_4
      inputName=None
      mask=0
      maskA=0
      maskB=0
      maskG=0
      maskR=0
      outputIndex=0
     }
    }
    Ocean {
     grassType=/Clyde/Landscape/GrassTypes/LandscapeGrassType_Ocean
     input {
      expression=&self:MaterialExpressionSubtract_5
      inputName=None
      mask=0
      maskA=0
      maskB=0
      maskG=0
      maskR=0
      outputIndex=0
     }
    }
   ]}
  LandscapeLayerSample() {parameterName=BiomeD}
  LandscapeLayerSample() {parameterName=BiomeD}
  LandscapeLayerSample() {parameterName=BiomeD}
  n57 = ComponentMask(None=WorldPosition().XYZ) {b=true}
  n58 = Constant() {r=4000}
  n59 = Saturate(None=Divide(A=Add(A=@n57, B=@n58), B=Constant() {r=500}))
  n60 = Add(A=@n57, B=Constant() {r=3990})
  Multiply(A=LandscapeLayerSample() {parameterName=Mud}, B=Multiply(A=OneMinus(None=@n59), B=Saturate(None=@n60)))
  NamedRerouteDeclaration(None=@n9.Swamp) {name=MountainTerrainBiomeMask, nodeColor=<rgba 1 0.4219 0 1>}
  Reroute(None=@n9.Swamp)
  Reroute(None=@n9.Swamp)
  n61 = BlendMaterialAttributes(A=@n34.Result, B=@n35.Material, Alpha=@n35.Material)
  Reroute(None=@n61)
  n62 = MaterialFunctionCall(UseLandscapeCoords=StaticBool() {value=true}) {materialFunction=&MF_ApolloTerrainMasks}
  Reroute(None=@n62.Swamp)
  Reroute(None=@n62.Forest)
  n63 = LandscapeLayerSample() {parameterName=BiomeB}
  n64 = LandscapeLayerSample() {parameterName=BiomeA}
  n65 = MaterialFunctionCall(
   UseBiomeColorMask_1=StaticBool() {value=true}
   BiomeMask_1=@n62.Swamp
   GrassBorderColorMask_1=Max(A=Max(A=@n64, B=@n63), B=@n73.MaskForGrassBorderColor)
   HeightmapFromCliff_1=@n73.HeightmapForGrass
   Grass-EdgeBiomeColorB_1=VectorParameter() {parameterName=GrassEdgeColorB_Swamp, defaultValue=<rgba 0.16 0.1163 0.016 1>}.RGB
   Grass-EdgeBiomeColorA_1=VectorParameter() {parameterName=GrassEdgeColorA_Swamp, defaultValue=<rgba 0.1 0.0727 0.01 1>}.RGB
   GravelBiomeColorDiv_1=VectorParameter() {parameterName=Biome-GravelDiv_Swamp, defaultValue=<rgba 0.8 0.5869 0.4049 1>}.RGB
   GravelBiomeColor_1=VectorParameter() {parameterName=Biome-GravelColor_Swamp, defaultValue=<rgba 0.35 0.3059 0.223 1>}.RGB
   MudBiomeColorDiv_1=VectorParameter() {parameterName=Biome-MudDiv_Swamp, defaultValue=<rgba 0.8 0.5869 0.4049 1>}.RGB
   MudBiomeColor_1=VectorParameter() {parameterName=Biome-MudColor_Swamp, defaultValue=<rgba 0.4219 0.2696 0.0879 1>}.RGB
   MudBiomeRoughMod_1=ScalarParameter() {defaultValue=0.9, parameterName=Biome-ModRoughMod_Swamp}
   SandBiomeColorDiv_1=VectorParameter() {parameterName=Biome-SandDiv_Swamp, defaultValue=<rgba 1 0.896 0.725 1>}.RGB
   SandBiomeColor_1=VectorParameter() {parameterName=Biome-SandColor_Swamp, defaultValue=<rgba 0.7188 0.5259 0.2957 1>}.RGB
  ) {materialFunction=&MF_CommonLandscapeLayers}
  n66 = MaterialFunctionCall(
   Material 1=@n65.Material
   Height Texture 1=@n65.Heightmap
   Material 2=@n65.Material
   Height Texture 2=@n65.Heightmap
   Transistion Phase=LandscapeLayerSample() {parameterName=NoFoliage}
  ) {materialFunction=&MF_BlendLandscapeLayers}
  n67 = MaterialFunctionCall!(
   Diffuse=@n65.GravelDiffuse_1
   Normals=@n65.GravelDiffuse_1
   Height=@n65.GravelDiffuse_1
   BiomeColorDiv=VectorParameter() {parameterName=Biome-WetMudDiv, defaultValue=<rgba 0.85 0.6208 0.425 1>}.RGB
   BiomeColor=VectorParameter() {parameterName=Biome-WetMudColor, defaultValue=<rgba 0.8229 0.4814 0.21 1>}.RGB
   BiomeColorDark=VectorParameter() {parameterName=Biome-WetMudColorDark, defaultValue=<rgba 0.7031 0.4014 0.0549 1>}.RGB
  ) {materialFunction=/Clyde/Athena/Apollo/Environments/Landscape/Materials/Biome/UniqueBiomeMaterialFunctions/MF_Biome_WetMud}
  n68 = MaterialFunctionCall(
   Material 1=@n66.BlendedMaterial
   Height Texture 1=@n66.BlendedMaterial
   Material 2=@n67.Material
   Height Texture 2=@n67.Material
   Transistion Phase=@n64
  ) {materialFunction=&MF_BlendLandscapeLayers}
  n69 = MaterialFunctionCall() {materialFunction=&MF_Biome_Forest}
  n70 = MaterialFunctionCall(
   Material 1=@n68.BlendedMaterial
   Height Texture 1=@n68.BlendedMaterial
   Material 2=@n69.Material
   Height Texture 2=@n69.Material
   Transistion Phase=@n63
  ) {materialFunction=&MF_BlendLandscapeLayers}
  n71 = MaterialFunctionCall() {materialFunction=/Clyde/Athena/Apollo/Environments/Landscape/Materials/Biome/UniqueBiomeMaterialFunctions/MF_Biome_SwampGrass}
  n72 = MaterialFunctionCall(
   Material 1=@n70.BlendedMaterial
   Height Texture 1=@n70.BlendedMaterial
   Material 2=@n71.Material
   Height Texture 2=@n71.Material
   Transistion Phase=LandscapeLayerSample() {parameterName=BiomeC}
  ) {materialFunction=&MF_BlendLandscapeLayers}
  n73 = MaterialFunctionCall(BiomeColorMask=@n62.Swamp, Material=@n72.BlendedMaterial, Heightmap=@n72.BlendedMaterial, UseBiomeColorMask=StaticBool()) {materialFunction=&MF_SlopeBlendCliff}
  n74 = BlendMaterialAttributes(
   A=@n61
   B=MaterialFunctionCall(Material=@n73.Material) {materialFunction=&TransformTangentToWorld}.Result
   Alpha=MaterialFunctionCall(UseLandscapeCoords=StaticBool() {value=true}) {materialFunction=&MF_ApolloTerrainMasks}.Swamp
  )
  n75 = GetMaterialAttributes(None=@n74) {attributeGetTypes [
    #1
    #2
    #3
    #4
   ]}
  RuntimeVirtualTextureOutput!(
   BaseColor=MaterialFunctionCall!(BaseColor=@n75.Base Color) {materialFunction=/Clyde/Athena/Apollo/Environments/Landscape/Materials/Biome/CommonMaterials/MF_ReduceRTV-CompressionArtifacts}.BaseColor
   Specular=@n75.Specular
   Roughness=@n75.Specular
   Normal=@n75.Specular
   WorldHeight=ComponentMask(None=WorldPosition().XYZ) {b=true}
   Opacity
   Mask
   Displacement
   Mask4
  )
  n76 = GetMaterialAttributes(None=MaterialFunctionCall!(UseSnow=StaticBool() {value=true}) {materialFunction=/Clyde/Athena/Apollo/Environments/Landscape/Materials/Biome/CommonMaterials/Mobile/MF_Mobile_CommonLandscapeLayers}.Material) {attributeGetTypes [#1]}
  MaterialExpressionConstant_13 = Constant()
  n77 = LandscapeLayerBlend!(
   Layer Gravel=@MaterialExpressionConstant_13
   Layer Mud=@MaterialExpressionConstant_13
   Layer Sand=@MaterialExpressionConstant_13
   Layer Asphalt=@MaterialExpressionConstant_13
   Layer NoFoliage=@MaterialExpressionConstant_13
  ) {layers [
    {
     blendType=LB_WeightBlend
     constHeightInput=0
     layerName=Gravel
     previewWeight=0
     constLayerInput=<0 0 0>
     heightInput {expression=null, inputName=None, mask=0, maskA=0, maskB=0, maskG=0, maskR=0, outputIndex=0}
     layerInput {
      expression=&self:MaterialExpressionConstant_13
      inputName=None
      mask=0
      maskA=0
      maskB=0
      maskG=0
      maskR=0
      outputIndex=0
     }
    }
    {
     blendType=LB_WeightBlend
     constHeightInput=0
     layerName=Mud
     previewWeight=0
     constLayerInput=<0 0 0>
     heightInput {expression=null, inputName=None, mask=0, maskA=0, maskB=0, maskG=0, maskR=0, outputIndex=0}
     layerInput {
      expression=&self:MaterialExpressionConstant_13
      inputName=None
      mask=0
      maskA=0
      maskB=0
      maskG=0
      maskR=0
      outputIndex=0
     }
    }
    {
     blendType=LB_WeightBlend
     constHeightInput=0
     layerName=Sand
     previewWeight=0
     constLayerInput=<0 0 0>
     heightInput {expression=null, inputName=None, mask=0, maskA=0, maskB=0, maskG=0, maskR=0, outputIndex=0}
     layerInput {
      expression=&self:MaterialExpressionConstant_13
      inputName=None
      mask=0
      maskA=0
      maskB=0
      maskG=0
      maskR=0
      outputIndex=0
     }
    }
    {
     blendType=LB_WeightBlend
     constHeightInput=0
     layerName=Asphalt
     previewWeight=0
     constLayerInput=<0 0 0>
     heightInput {expression=null, inputName=None, mask=0, maskA=0, maskB=0, maskG=0, maskR=0, outputIndex=0}
     layerInput {
      expression=&self:MaterialExpressionConstant_13
      inputName=None
      mask=0
      maskA=0
      maskB=0
      maskG=0
      maskR=0
      outputIndex=0
     }
    }
    {
     blendType=LB_WeightBlend
     constHeightInput=0
     layerName=NoFoliage
     previewWeight=0
     constLayerInput=<0 0 0>
     heightInput {expression=null, inputName=None, mask=0, maskA=0, maskB=0, maskG=0, maskR=0, outputIndex=0}
     layerInput {
      expression=&self:MaterialExpressionConstant_13
      inputName=None
      mask=0
      maskA=0
      maskB=0
      maskG=0
      maskR=0
      outputIndex=0
     }
    }
   ]}
  n78 = SetMaterialAttributes!(MaterialAttributes=@n76.MaterialAttributes, Base Color=Add(A=@n76.Base Color, B=Multiply(A=@n77) {constB=0})) {attributeSetTypes [#1]}
  n79 = RuntimeVirtualTextureSampleParameter() {
   group=RVT
   materialType=BaseColor_Normal_Specular_YCoCg
   parameterName=RVT Sample Target
   virtualTexture=/Game/Athena/Apollo/RVT/RVT_Apollo_Landscape
  }
  n80 = SetMaterialAttributes!(MaterialAttributes, Base Color=@n79.BaseColor, Specular=@n79.BaseColor, Roughness=@n79.BaseColor, Normal=@n79.BaseColor) {attributeSetTypes [
    #1
    #2
    #3
    #4
   ]}
  n81 = StaticSwitchParameter(
   True=MaterialFunctionCall!(RVT=@n80, ByPass RVT=@n74) {materialFunction=/Clyde/Athena/Apollo/Maps/Landscape/VT/MaterialFunctions/EnableRVT}.Result
   False=@n74
  ) {defaultValue=true, group=RVT Base, parameterName=RVT Enable}
  SetMaterialAttributes!(
   MaterialAttributes=MaterialFunctionCall!(
    Material=ShadingPathSwitch(Default=@n81, Deferred=@n81, Forward=@n78, Mobile=@n78)
    DebugInputAsBasecolor
    Non-RVT material
    DebugMaterialAs1Color=VectorParameter() {parameterName=DebugBiomeColor, defaultValue=<rgba 0 0.2359 0.474 1>}.RGB
   ) {materialFunction=/Game/Athena/Apollo/Environments/Landscape/Materials/Biome/CommonMaterials/MF_LS_DEBUG}.Result
   Base Color=MaterialFunctionCall(UseLandscapeCoords=StaticBool()) {materialFunction=&MF_ApolloTerrainMasks}.BiomeMaskRGBA
  ) {attributeSetTypes [#1]}
  StaticBool() {value=true}
  VectorParameter() {parameterName=Biome-CliffDiv, defaultValue=<rgba 1 1 1 1>}
  VectorParameter() {parameterName=Biome-CliffColor, defaultValue=<rgba 1 0.96 0.9 1>}
  VectorParameter() {parameterName=Biome-CliffDiv, defaultValue=<rgba 1 1 1 1>}
  VectorParameter() {parameterName=Biome-CliffColor, defaultValue=<rgba 1 0.96 0.9 1>}
 ]
}
