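# Landscape material for the Apollo swamp biome, with Nanite usage enabled.
# Common layers are blended with wet mud, forest and swamp grass by painted BiomeA/B/C layers, then passed through slope-based cliff blending.
# Grass output derives foliage from layer and terrain masks, minus NoFoliage, limiting reeds to mud within a world-height band.
# The result writes to and optionally samples a runtime virtual texture, with simpler mobile layers as fallback, before a debug function.
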Material /Clyde/Athena/Apollo/Environments/Landscape/Materials/Biome/LandscapeBiomeMaterials/M_Biome_Swamp
graph {
 properties {
  bTangentSpaceNormal=false
  bUseMaterialAttributes=true
  bUsedWithNanite=true
  pixelDepthOffsetMode=PDOM_Legacy
  usedShadingModels=Default Lit
 }
 nodes [
  Constant() {r=3909}
  n1 = Max(A=LandscapeLayerSample() {parameterName=BiomeA}, B=LandscapeLayerSample() {parameterName=BiomeB})
  n2 = LandscapeLayerSample() {parameterName=BiomeC}
  n3 = Saturate(None=Multiply(A=Constant() {r=3}, B=@n2))
  n4 = Constant() {r=2}
  n5 = LandscapeLayerSample() {parameterName=Mud}
  n6 = Max(A=Max(A=LandscapeLayerSample() {parameterName=Gravel}, B=LandscapeLayerSample() {parameterName=Sand}), B=@n5)
  n7 = LandscapeLayerSample() {parameterName=Asphalt}
  n8 = Saturate(None=Max(A=Multiply(A=@n4, B=@n6), B=Ceil(None=@n7)))
  n9 = MaterialFunctionCall!(UseLandscapeCoords=StaticBool() {value=true}) {materialFunction=/Clyde/Athena/Apollo/Environments/Landscape/Materials/Biome/CommonMaterials/MF_ApolloTerrainMasks}
  n10 = LandscapeLayerSample() {parameterName=BiomeB}
  n11 = LandscapeLayerSample() {parameterName=BiomeA}
  n12 = MaterialFunctionCall!(
   UseBiomeColorMask_1=StaticBool() {value=true}
   BiomeMask_1=@n9.Swamp
   GrassBorderColorMask_1=Max(A=Max(A=@n11, B=@n10), B=@n20.MaskForGrassBorderColor)
   HeightmapFromCliff_1=@n20.HeightmapForGrass
   Grass-EdgeBiomeColorB_1=VectorParameter() {parameterName=GrassEdgeColorB, defaultValue=<rgba 0.16 0.1163 0.016 1>}.RGB
   Grass-EdgeBiomeColorA_1=VectorParameter() {parameterName=GrassEdgeColorA, defaultValue=<rgba 0.1 0.0727 0.01 1>}.RGB
   GravelBiomeColorDiv_1=VectorParameter() {parameterName=Biome-GravelDiv, defaultValue=<rgba 0.8 0.5869 0.4049 1>}.RGB
   GravelBiomeColor_1=VectorParameter() {parameterName=Biome-GravelColor, defaultValue=<rgba 0.35 0.3059 0.223 1>}.RGB
   MudBiomeColorDiv_1=VectorParameter() {parameterName=Biome-MudDiv, defaultValue=<rgba 0.8 0.5869 0.4049 1>}.RGB
   MudBiomeColor_1=VectorParameter() {parameterName=Biome-MudColor, defaultValue=<rgba 0.4219 0.2696 0.0879 1>}.RGB
   MudBiomeRoughMod_1=ScalarParameter() {defaultValue=0.9, parameterName=Biome-ModRoughMod}
   SandBiomeColorDiv_1=VectorParameter() {parameterName=Biome-SandDiv, defaultValue=<rgba 1 0.896 0.725 1>}.RGB
   SandBiomeColor_1=VectorParameter() {parameterName=Biome-SandColor, defaultValue=<rgba 0.7188 0.5259 0.2957 1>}.RGB
  ) {materialFunction=/Clyde/Athena/Apollo/Environments/Landscape/Materials/Biome/CommonMaterials/MF_CommonLandscapeLayers}
  n13 = MaterialFunctionCall!(
   Material 1=@n12.Material
   Height Texture 1=@n12.Heightmap
   Material 2=@n12.Material
   Height Texture 2=@n12.Heightmap
   Transistion Phase=LandscapeLayerSample() {parameterName=NoFoliage}
   UseContrast
   Contrast
  ) {materialFunction=/Clyde/Athena/Apollo/Environments/Landscape/Materials/Biome/CommonMaterials/MF_BlendLandscapeLayers}
  n14 = MaterialFunctionCall!(
   Diffuse=@n12.GravelDiffuse_1
   Normals=@n12.GravelDiffuse_1
   Height=@n12.GravelDiffuse_1
   BiomeColorDiv=VectorParameter() {parameterName=Biome-WetMudDiv, defaultValue=<rgba 0.85 0.6208 0.425 1>}.RGB
   BiomeColor=VectorParameter() {parameterName=Biome-WetMudColor, defaultValue=<rgba 0.8229 0.4814 0.21 1>}.RGB
   BiomeColorDark=VectorParameter() {parameterName=Biome-WetMudColorDark, defaultValue=<rgba 0.7031 0.4014 0.0549 1>}.RGB
  ) {materialFunction=/Clyde/Athena/Apollo/Environments/Landscape/Materials/Biome/UniqueBiomeMaterialFunctions/MF_Biome_WetMud}
  n15 = MaterialFunctionCall(
   Material 1=@n13.BlendedMaterial
   Height Texture 1=@n13.BlendedMaterial
   Material 2=@n14.Material
   Height Texture 2=@n14.Material
   Transistion Phase=@n11
  ) {materialFunction=&MF_BlendLandscapeLayers}
  n16 = MaterialFunctionCall!(UseBiomeColorMask, BiomeColorMask, BiomeColorDiv, BiomeColor) {materialFunction=/Clyde/Athena/Apollo/Environments/Landscape/Materials/Biome/UniqueBiomeMaterialFunctions/MF_Biome_Forest}
  n17 = MaterialFunctionCall(
   Material 1=@n15.BlendedMaterial
   Height Texture 1=@n15.BlendedMaterial
   Material 2=@n16.Material
   Height Texture 2=@n16.Material
   Transistion Phase=@n10
  ) {materialFunction=&MF_BlendLandscapeLayers}
  n18 = MaterialFunctionCall() {materialFunction=/Clyde/Athena/Apollo/Environments/Landscape/Materials/Biome/UniqueBiomeMaterialFunctions/MF_Biome_SwampGrass}
  n19 = MaterialFunctionCall(
   Material 1=@n17.BlendedMaterial
   Height Texture 1=@n17.BlendedMaterial
   Material 2=@n18.Material
   Height Texture 2=@n18.Material
   Transistion Phase=LandscapeLayerSample() {parameterName=BiomeC}
  ) {materialFunction=&MF_BlendLandscapeLayers}
  n20 = MaterialFunctionCall!(
   BiomeColorMask=@n9.Swamp
   Material=@n19.BlendedMaterial
   Heightmap=@n19.BlendedMaterial
   UseBiomeColorMask=StaticBool()
   BiomeColorDiv
   BiomeColor
  ) {materialFunction=/Clyde/Athena/Apollo/Environments/Landscape/Materials/Biome/CommonMaterials/MF_SlopeBlendCliff}
  n21 = OneMinus(None=Max(A=Max(A=@n1, B=@n3), B=Max(A=@n20.LandscapeGrassMeshMask, B=@n8)))
  n22 = Subtract(A=@n21, B=@n9.Forest) {desc=Grass_Standard}
  MaterialExpressionSubtract_5 = Subtract(A=@n22, B=LandscapeLayerSample() {parameterName=NoFoliage})
  n23 = Min(A=@n9.Forest, B=@n21) {desc=Forest grass}
  MaterialExpressionSubtract_4 = Subtract(A=Max(A=Max(A=@n22, B=@n23), B=@n2), B=LandscapeLayerSample() {parameterName=NoFoliage})
  MaterialExpressionSubtract_7 = Subtract(A=Constant(), B=LandscapeLayerSample() {parameterName=NoFoliage})
  MaterialExpressionSubtract_3 = Subtract(A=@n2, B=LandscapeLayerSample() {parameterName=NoFoliage})
  n24 = ComponentMask(None=WorldPosition().XYZ) {b=true}
  n25 = Add(A=@n24, B=Constant() {r=4000})
  n26 = Constant() {r=500}
  n27 = OneMinus(None=Saturate(None=Divide(A=@n25, B=@n26)))
  n28 = Saturate(None=Add(A=@n24, B=Constant() {r=3990}))
  MaterialExpressionSubtract_2 = Subtract(A=Multiply(A=@n5, B=Multiply(A=@n27, B=@n28)), B=LandscapeLayerSample() {parameterName=NoFoliage})
  MaterialExpressionSubtract_1 = Subtract(A=@n23, B=LandscapeLayerSample() {parameterName=NoFoliage})
  LandscapeGrassOutput!(
   Grass_Standard=@MaterialExpressionSubtract_5
   Bush=@MaterialExpressionSubtract_4
   FarmFlowers=@MaterialExpressionSubtract_7
   Scree=@MaterialExpressionSubtract_7
   Scree_Grass=@MaterialExpressionSubtract_7
   Heather=@MaterialExpressionSubtract_7
   SwampGrass=@MaterialExpressionSubtract_3
   SwampReeds=@MaterialExpressionSubtract_2
   Forest_Grass=@MaterialExpressionSubtract_1
   Beach=@MaterialExpressionSubtract_7
   Ocean=@MaterialExpressionSubtract_7
  ) {grassTypes [
    Grass_Standard {
     grassType=/Clyde/Landscape/GrassTypes/LandscapeGrassType_StandardGrass
     input {
      expression=&self:MaterialExpressionSubtract_5
      inputName=None
      mask=0
      maskA=0
      maskB=0
      maskG=0
      maskR=0
      outputIndex=0
     }
    }
    Bush {
     grassType=/Clyde/Landscape/GrassTypes/Athena_Foliage_Bushes
     input {
      expression=&self:MaterialExpressionSubtract_4
      inputName=None
      mask=0
      maskA=0
      maskB=0
      maskG=0
      maskR=0
      outputIndex=0
     }
    }
    FarmFlowers {
     grassType=/Clyde/Landscape/GrassTypes/LandscapeGrassType_FarmFlowers
     input {
      expression=&self:MaterialExpressionSubtract_7
      inputName=None
      mask=0
      maskA=0
      maskB=0
      maskG=0
      maskR=0
      outputIndex=0
     }
    }
    Scree {
     grassType=/Clyde/Landscape/GrassTypes/LandscapeGrassType_Scree
     input {
      expression=&self:MaterialExpressionSubtract_7
      inputName=None
      mask=0
      maskA=0
      maskB=0
      maskG=0
      maskR=0
      outputIndex=0
     }
    }
    Scree_Grass {
     grassType=/Clyde/Landscape/GrassTypes/LandscapeGrassType_ScreeGrass
     input {
      expression=&self:MaterialExpressionSubtract_7
      inputName=None
      mask=0
      maskA=0
      maskB=0
      maskG=0
      maskR=0
      outputIndex=0
     }
    }
    Heather {
     grassType=/Clyde/Landscape/GrassTypes/LandscapeGrassType_Heather
     input {
      expression=&self:MaterialExpressionSubtract_7
      inputName=None
      mask=0
      maskA=0
      maskB=0
      maskG=0
      maskR=0
      outputIndex=0
     }
    }
    SwampGrass {
     grassType=/Clyde/Landscape/GrassTypes/LandscapeGrassType_SwampGrass
     input {
      expression=&self:MaterialExpressionSubtract_3
      inputName=None
      mask=0
      maskA=0
      maskB=0
      maskG=0
      maskR=0
      outputIndex=0
     }
    }
    SwampReeds {
     grassType=/Clyde/Landscape/GrassTypes/LandscapeGrassType_SwampReeds
     input {
      expression=&self:MaterialExpressionSubtract_2
      inputName=None
      mask=0
      maskA=0
      maskB=0
      maskG=0
      maskR=0
      outputIndex=0
     }
    }
    Forest_Grass {
     grassType=/Clyde/Landscape/GrassTypes/LandscapeGrassType_ForestGrass
     input {
      expression=&self:MaterialExpressionSubtract_1
      inputName=None
      mask=0
      maskA=0
      maskB=0
      maskG=0
      maskR=0
      outputIndex=0
     }
    }
    Beach {
     grassType=/Clyde/Landscape/GrassTypes/LandscapeGrassType_Beach
     input {
      expression=&self:MaterialExpressionSubtract_7
      inputName=None
      mask=0
      maskA=0
      maskB=0
      maskG=0
      maskR=0
      outputIndex=0
     }
    }
    Ocean {
     grassType=/Clyde/Landscape/GrassTypes/LandscapeGrassType_Ocean
     input {
      expression=&self:MaterialExpressionSubtract_7
      inputName=None
      mask=0
      maskA=0
      maskB=0
      maskG=0
      maskR=0
      outputIndex=0
     }
    }
   ]}
  LandscapeLayerSample() {parameterName=BiomeD}
  LandscapeLayerSample() {parameterName=BiomeD}
  n29 = GetMaterialAttributes(None=MaterialFunctionCall!(UseSnow=StaticBool()) {materialFunction=/Clyde/Athena/Apollo/Environments/Landscape/Materials/Biome/CommonMaterials/Mobile/MF_Mobile_CommonLandscapeLayers}.Material) {attributeGetTypes [#1]}
  MaterialExpressionConstant_13 = Constant()
  n30 = LandscapeLayerBlend!(
   Layer Gravel=@MaterialExpressionConstant_13
   Layer Mud=@MaterialExpressionConstant_13
   Layer Sand=@MaterialExpressionConstant_13
   Layer Asphalt=@MaterialExpressionConstant_13
   Layer NoFoliage=@MaterialExpressionConstant_13
  ) {layers [
    {
     blendType=LB_WeightBlend
     constHeightInput=0
     layerName=Gravel
     previewWeight=0
     constLayerInput=<0 0 0>
     heightInput {expression=null, inputName=None, mask=0, maskA=0, maskB=0, maskG=0, maskR=0, outputIndex=0}
     layerInput {
      expression=&self:MaterialExpressionConstant_13
      inputName=None
      mask=0
      maskA=0
      maskB=0
      maskG=0
      maskR=0
      outputIndex=0
     }
    }
    {
     blendType=LB_WeightBlend
     constHeightInput=0
     layerName=Mud
     previewWeight=0
     constLayerInput=<0 0 0>
     heightInput {expression=null, inputName=None, mask=0, maskA=0, maskB=0, maskG=0, maskR=0, outputIndex=0}
     layerInput {
      expression=&self:MaterialExpressionConstant_13
      inputName=None
      mask=0
      maskA=0
      maskB=0
      maskG=0
      maskR=0
      outputIndex=0
     }
    }
    {
     blendType=LB_WeightBlend
     constHeightInput=0
     layerName=Sand
     previewWeight=0
     constLayerInput=<0 0 0>
     heightInput {expression=null, inputName=None, mask=0, maskA=0, maskB=0, maskG=0, maskR=0, outputIndex=0}
     layerInput {
      expression=&self:MaterialExpressionConstant_13
      inputName=None
      mask=0
      maskA=0
      maskB=0
      maskG=0
      maskR=0
      outputIndex=0
     }
    }
    {
     blendType=LB_WeightBlend
     constHeightInput=0
     layerName=Asphalt
     previewWeight=0
     constLayerInput=<0 0 0>
     heightInput {expression=null, inputName=None, mask=0, maskA=0, maskB=0, maskG=0, maskR=0, outputIndex=0}
     layerInput {
      expression=&self:MaterialExpressionConstant_13
      inputName=None
      mask=0
      maskA=0
      maskB=0
      maskG=0
      maskR=0
      outputIndex=0
     }
    }
    {
     blendType=LB_WeightBlend
     constHeightInput=0
     layerName=NoFoliage
     previewWeight=0
     constLayerInput=<0 0 0>
     heightInput {expression=null, inputName=None, mask=0, maskA=0, maskB=0, maskG=0, maskR=0, outputIndex=0}
     layerInput {
      expression=&self:MaterialExpressionConstant_13
      inputName=None
      mask=0
      maskA=0
      maskB=0
      maskG=0
      maskR=0
      outputIndex=0
     }
    }
   ]}
  n31 = SetMaterialAttributes!(MaterialAttributes=@n29.MaterialAttributes, Base Color=Add(A=@n29.Base Color, B=Multiply(A=@n30) {constB=0})) {attributeSetTypes [#1]}
  n32 = RuntimeVirtualTextureSampleParameter() {
   group=RVT
   materialType=BaseColor_Normal_Specular_YCoCg
   parameterName=RVT Sample Target
   virtualTexture=/Game/Athena/Apollo/RVT/RVT_Apollo_Landscape
  }
  n33 = SetMaterialAttributes!(MaterialAttributes, Base Color=@n32.BaseColor, Specular=@n32.BaseColor, Roughness=@n32.BaseColor, Normal=@n32.BaseColor) {attributeSetTypes [
    #1
    #2
    #3
    #4
   ]}
  n34 = MaterialFunctionCall!(Material=@n20.Material) {materialFunction=/Game/Athena/Apollo/Maps/Landscape/VT/MaterialFunctions/TransformTangentToWorld}
  n35 = VirtualTextureFeatureSwitch!(
   No=@n31
   Yes=StaticSwitchParameter(
    True=MaterialFunctionCall!(RVT=@n33, ByPass RVT=@n34.Result) {materialFunction=/Clyde/Athena/Apollo/Maps/Landscape/VT/MaterialFunctions/EnableRVT}.Result
    False=@n34.Result
   ) {defaultValue=true, group=RVT Base, parameterName=RVT Enable}
  )
  MaterialFunctionCall!(
   Material=ShadingPathSwitch(Default=@n35, Mobile=QualitySwitch(Default=@n31, Medium=@n35))
   DebugInputAsBasecolor
   Non-RVT material
   DebugMaterialAs1Color=VectorParameter() {parameterName=DebugBiomeColor, defaultValue=<rgba 0.4346 0 0.474 1>}.RGB
  ) {materialFunction=/Game/Athena/Apollo/Environments/Landscape/Materials/Biome/CommonMaterials/MF_LS_DEBUG}
  Reroute(None=@n9.Swamp) {desc=Swamp Mask}
  n36 = GetMaterialAttributes(None=@n34.Result) {attributeGetTypes [
    #1
    #2
    #3
    #4
   ]}
  RuntimeVirtualTextureOutput!(
   BaseColor=MaterialFunctionCall!(BaseColor=@n36.Base Color) {materialFunction=/Clyde/Athena/Apollo/Environments/Landscape/Materials/Biome/CommonMaterials/MF_ReduceRTV-CompressionArtifacts}.BaseColor
   Specular=@n36.Specular
   Roughness=@n36.Specular
   Normal=@n36.Specular
   WorldHeight=ComponentMask(None=WorldPosition().XYZ) {b=true}
   Opacity
   Mask
   Displacement
   Mask4
  )
  VectorParameter() {parameterName=Biome-CliffDiv, defaultValue=<rgba 1 1 1 1>}
  VectorParameter() {parameterName=Biome-CliffColor, defaultValue=<rgba 1 1 1 1>}
 ]
}
