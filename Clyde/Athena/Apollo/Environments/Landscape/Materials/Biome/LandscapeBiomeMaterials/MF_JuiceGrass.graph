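# Landscape material function that blends purple-tinted "juice" moss over incoming material attributes inside the SlurpyCenter/SlurpyRadius sphere from FortniteMaterialParameters, suppressed on Mud, Sand and Asphalt layers.
# The moss combines world-space color masks, curve-atlas detail, Fresnel fuzz, bump-offset textures and layered normals.
# Biome color division adjusts base color, with edge colors near other landscape layers.
# Outputs full Material, simplified Mobile-Material and moss-texture Height.
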MaterialFunction /Clyde/Athena/Apollo/Environments/Landscape/Materials/Biome/LandscapeBiomeMaterials/MF_JuiceGrass
graph {nodes [
  Constant3Vector() {constant=<rgba 1.0666 0.3094 0.5924 1>}
  Constant() {r=500}
  Constant() {r=4560}
  Constant() {r=13000}
  Constant() {r=5599}
  n1 = FunctionInput() {inputType=FunctionInput_MaterialAttributes}
  n2 = SphereMask(
   A=WorldPosition().XYZ
   B=ComponentMask(None=CollectionParameter() {collection=&FortniteMaterialParameters, parameterName=SlurpyCenter}) {b=true, g=true, r=true}
   Radius=ComponentMask(None=CollectionParameter() {collection=&FortniteMaterialParameters, parameterName=SlurpyRadius}) {b=true, g=true, r=true}
   Hardness=ScalarParameter() {defaultValue=0.9, parameterName=SlurpyGradient}
  )
  n3 = LandscapeLayerSample() {parameterName=Mud}
  n4 = Max(A=Max(A=@n3, B=LandscapeLayerSample() {parameterName=Sand}), B=LandscapeLayerSample() {parameterName=Asphalt})
  n5 = Constant() {r=2}
  n6 = Saturate(None=Subtract(A=Clamp(None=@n2), B=Multiply(A=@n4, B=@n5)))
  FunctionOutput(None=BlendMaterialAttributes(A=@n1, B=NamedRerouteUsage(), Alpha=@n6) {vertexAttributeBlendType=UseA}) {outputName=Material}
  n7 = LandscapeLayerCoords() {mappingScale=10}
  n8 = Multiply(A=@n7, B=ScalarParameter() {defaultValue=2, group=FrontierMoss, parameterName=Moss_Scale})
  n9 = TextureSample(UVs=@n8) {samplerType=SAMPLERTYPE_Masks, texture=/Game/Environments/Artemis/Sets/Terrain/Textures/T_Artemis_Frontier_Moss_01_D}
  FunctionOutput(None=@n9.A) {outputName=Height}
  n10 = NamedRerouteUsage()
  n11 = Multiply(A=@n10, B=ScalarParameter() {defaultValue=0.25, group=FrontierMoss, parameterName=Moss_CoreDarkening})
  n12 = Multiply(A=@n10, B=ScalarParameter() {defaultValue=1.5, group=FrontierMoss, parameterName=Moss_FuzzyEdgeBrightness})
  n13 = ScalarParameter() {defaultValue=1, group=FrontierMoss, parameterName=FuzzExponent}
  n14 = LinearInterpolate(
   B=Power(Base=Fresnel(ExponentIn=@n13), Exp=ScalarParameter() {defaultValue=2, group=FrontierMoss, parameterName=FuzzEdgeClamp})
   Alpha=ScalarParameter() {defaultValue=0.2, group=FrontierMoss, parameterName=Mobile-FuzzBlend}
  ) {constAlpha=0.2}
  n15 = SetMaterialAttributes!(
   MaterialAttributes
   BaseColor=Multiply(A=LinearInterpolate(A=@n11, B=@n12, Alpha=@n14), B=VectorParameter() {parameterName=Mobile-JuiceColor, defaultValue=<rgba 1 1 1 1>}.RGB)
   Specular=LinearInterpolate(
    A=ScalarParameter() {defaultValue=0.5, group=FrontierMoss, parameterName=Moss_Min_Specular}
    B=ScalarParameter() {defaultValue=1, group=FrontierMoss, parameterName=Moss_Max_Specular}
   )
   Roughness=LinearInterpolate(
    A=ScalarParameter() {defaultValue=0.75, group=FrontierMoss, parameterName=Moss_Min_Roughness}
    B=ScalarParameter() {defaultValue=1, group=FrontierMoss, parameterName=Moss_Max_Roughness}
   )
   Normal=VertexNormalWS()
  ) {attributeSetTypes [
    #1
    #2
    #3
    #4
   ]}
  FunctionOutput(None=BlendMaterialAttributes(A=@n1, B=@n15, Alpha=@n6) {vertexAttributeBlendType=UseA}) {outputName=Mobile-Material}
  n16 = Constant3Vector() {constant=<rgba 0.6654 0.2918 0.5972 1>}
  n17 = WorldPosition() {worldPositionShaderOffset=WPT_ExcludeAllShaderOffsets}
  n18 = TextureSample(UVs=Divide(A=ComponentMask(None=@n17.XYZ) {g=true, r=true}, B=Constant() {r=-13000})) {
   samplerSource=SSM_Wrap_WorldGroupSettings
   samplerType=SAMPLERTYPE_Grayscale
   texture=/Game/Athena/Apollo/Environments/Landscape/Materials/Biome/CommonMaterials/T_ApolloGrass_ColorMask
  }
  LinearInterpolate(A=Constant3Vector() {constant=<rgba 0.581 0.2737 0.9062 1>}, B=@n16, Alpha=@n18.R)
  n19 = ScalarParameter() {defaultValue=1, group=FrontierMoss, parameterName=FuzzNormalScale}
  n20 = MaterialFunctionCall(
   Normal=TextureSample(UVs=Multiply(A=@n7, B=@n19)) {samplerType=SAMPLERTYPE_Normal, texture=/Game/Packages/Fortress_Sky/Materials/2Normal}.RGB
   Flatness=ScalarParameter() {defaultValue=0.5, group=FrontierMoss, parameterName=MAcroNormalFlatter}
  ) {materialFunction=&FlattenNormal}
  n21 = Power(
   Base=Fresnel(ExponentIn=ScalarParameter() {defaultValue=1, group=FrontierMoss, parameterName=FuzzExponent}, Normal=Transform(None=@n20.Result))
   Exp=ScalarParameter() {defaultValue=2, group=FrontierMoss, parameterName=FuzzEdgeClamp}
  )
  n22 = TextureSample(UVs=Multiply(A=@n7, B=ScalarParameter() {defaultValue=1, parameterName=MossLumpTiling})) {
   samplerType=SAMPLERTYPE_Grayscale
   texture=&T_ApolloGrass_ColorMask
  }
  Max(A=@n21, B=@n22.G)
  Max(A=@n3, B=Divide(B=ScalarParameter() {defaultValue=1, parameterName=JuiceGravelStrength, sliderMin=0.01}) {constB=4})
  NamedRerouteDeclaration(None=@n16) {name=MacroColorLerp, nodeColor=<rgba 0 0.2266 1 1>}
  n23 = Constant() {r=1.5}
  n24 = Constant() {r=-9}
  n25 = ComponentMask(None=WorldPosition() {worldPositionShaderOffset=WPT_ExcludeAllShaderOffsets}.XYZ) {g=true, r=true}
  n26 = TextureSample(UVs=Divide(A=@n25, B=Constant() {r=-3600})) {
   samplerSource=SSM_Wrap_WorldGroupSettings
   samplerType=SAMPLERTYPE_Grayscale
   texture=&T_ApolloGrass_ColorMask
  }
  n27 = LinearInterpolate(
   A=VectorParameter() {group=FrontierMoss, parameterName=Moss_Color_01, defaultValue=<rgba 0.8398 0.3833 1.0666 1>}.RGB
   B=VectorParameter() {group=FrontierMoss, parameterName=Moss_Color_02, defaultValue=<rgba 0.6574 0.3958 0.6666 1>}.RGB
   Alpha=Saturate(None=LinearInterpolate(A=@n23, B=@n24, Alpha=@n26.A))
  )
  n28 = VectorParameter() {group=FrontierMoss, parameterName=Moss_Color_03, defaultValue=<rgba 0.8398 0.3833 1.0666 1>}
  n29 = Constant() {r=500}
  n30 = TextureSample(UVs=Divide(A=Add(A=@n25, B=@n29), B=Constant() {r=4560})) {
   samplerSource=SSM_Wrap_WorldGroupSettings
   samplerType=SAMPLERTYPE_Grayscale
   texture=&T_ApolloGrass_ColorMask
  }
  n31 = Saturate(None=LinearInterpolate(A=Constant() {r=-0.6}, B=Constant() {r=2}, Alpha=@n30.A))
  NamedRerouteDeclaration(None=LinearInterpolate(
    A=LinearInterpolate(A=@n27, B=@n28.RGB, Alpha=@n31)
    B=VectorParameter() {group=FrontierMoss, parameterName=Moss_Color_04, defaultValue=<rgba 0.6574 0.3958 0.6666 1>}.RGB
    Alpha=@n26.A
   )) {name=MossMicroColors, nodeColor=<rgba 1 0.5859 0 1>}
  n32 = CurveAtlasRowParameter(CurveTime=LinearInterpolate(A=ScalarParameter() {defaultValue=0.25, group=FrontierMoss, parameterName=DiffuseMin}, Alpha=@n9.R)) {
   atlas=/Game/Athena/Apollo/Environments/Landscape/Materials/Frontier/FrontierColorAtlas
   curve=/Game/Athena/Apollo/Environments/Landscape/Materials/Frontier/FrontierMossDetailColor
   group=FrontierMoss
   parameterName=FrontierMossDetailColor
  }
  NamedRerouteDeclaration(None=@n32) {name=MossBaseColorDetail, nodeColor=<rgba 1 0 0.6562 1>}
  n33 = BumpOffset(
   Coordinate=@n8
   Height=TextureSample(UVs=@n8) {samplerType=SAMPLERTYPE_Masks, texture=&T_Artemis_Frontier_Moss_01_D}.R
   HeightRatioInput=ScalarParameter() {defaultValue=-0.1, group=FrontierMoss, parameterName=TerrainHeightRatio}
  )
  n34 = TextureSample(UVs=@n33) {samplerType=SAMPLERTYPE_Masks, texture=/Game/Environments/Artemis/Sets/Terrain/Textures/T_Artemis_Frontier_Moss_01_S}
  n35 = TextureSample(UVs=@n33) {samplerType=SAMPLERTYPE_Normal, texture=/Game/Environments/Artemis/Sets/Terrain/Textures/T_Artemis_Frontier_Moss_01_N}
  n36 = ScalarParameter() {defaultValue=0.25, group=FrontierMoss, parameterName=MicroNormalFlatter}
  n37 = VectorParameter() {parameterName=JuiceEdge1, defaultValue=<rgba 0.18 0.1308 0.018 1>}
  n38 = VectorParameter() {parameterName=JuiceEdge2, defaultValue=<rgba 0.2 0.1454 0.02 1>}
  n39 = WorldPosition() {worldPositionShaderOffset=WPT_ExcludeAllShaderOffsets}
  n40 = TextureSample(UVs=Divide(A=ComponentMask(None=@n39.XYZ) {g=true, r=true}, B=ScalarParameter() {defaultValue=-13000, parameterName=fet654644})) {
   samplerSource=SSM_Wrap_WorldGroupSettings
   samplerType=SAMPLERTYPE_Grayscale
   texture=&T_ApolloGrass_ColorMask
  }
  n41 = Max(
   A=Max(A=LandscapeLayerSample() {parameterName=Mud}, B=LandscapeLayerSample() {parameterName=Gravel})
   B=LandscapeLayerSample() {parameterName=Sand}
  )
  n42 = LandscapeLayerSample() {parameterName=Asphalt}
  n43 = Constant()
  n44 = Multiply(
   A=Max(
    A=Max(A=@n41, B=@n42)
    B=FunctionInput(Preview=@n43) {bUsePreviewValueAsDefault=true, inputName=GrassBorderColorMask_1, inputType=FunctionInput_Scalar, sortPriority=-8}
   )
   B=Constant() {r=2}
  )
  n45 = FunctionInput(Preview=StaticBool()) {bUsePreviewValueAsDefault=true, inputName=UseBiomeColorMask, inputType=FunctionInput_StaticBool, previewValue=<64 64 64 6>}
  n46 = FunctionInput() {bUsePreviewValueAsDefault=true, inputName=BiomeColorDiv, sortPriority=2, previewValue=<1 1 1 1>}
  n47 = Constant() {r=0.01}
  n48 = NamedRerouteUsage()
  n49 = NamedRerouteUsage()
  n50 = Saturate(None=Add(A=Multiply(A=@n47, B=@n48), B=Multiply(A=@n48, B=@n49)))
  n51 = ScalarParameter() {defaultValue=0.25, group=FrontierMoss, parameterName=Moss_CoreDarkening}
  n52 = ScalarParameter() {defaultValue=1.5, group=FrontierMoss, parameterName=Moss_FuzzyEdgeBrightness}
  n53 = MaterialFunctionCall!(
   UseBiomeColorMask=@n45
   BiomeColorMask=FunctionInput(Preview=Constant() {r=1}) {bUsePreviewValueAsDefault=true, inputName=BiomeColorMask, inputType=FunctionInput_Scalar, sortPriority=1}
   BiomeColorDiv=@n46
   BiomeColor=FunctionInput() {bUsePreviewValueAsDefault=true, inputName=BiomeColor, sortPriority=3, previewValue=<1 1 1 1>}
   Basecolor=LinearInterpolate(A=Multiply(A=@n50, B=@n51), B=Multiply(A=@n50, B=@n52), Alpha=Clamp(None=@n21))
  ) {materialFunction=/Clyde/Athena/Apollo/Environments/Landscape/Materials/Biome/CommonMaterials/MF_LS_DivideColor}
  n54 = SetMaterialAttributes(
   BaseColor=LinearInterpolate(A=Saturate(None=@n53.Result), B=LinearInterpolate(A=@n37.RGB, B=@n38.RGB, Alpha=@n40.R), Alpha=Saturate(None=@n44))
   Specular=LinearInterpolate(
    A=ScalarParameter() {defaultValue=0.5, group=FrontierMoss, parameterName=Moss_Min_Specular}
    B=ScalarParameter() {defaultValue=1, group=FrontierMoss, parameterName=Moss_Max_Specular}
    Alpha=@n34.R
   )
   Roughness=LinearInterpolate(
    A=ScalarParameter() {defaultValue=0.75, group=FrontierMoss, parameterName=Moss_Min_Roughness}
    B=ScalarParameter() {defaultValue=1, group=FrontierMoss, parameterName=Moss_Max_Roughness}
    Alpha=@n34.B
   )
   Normal=MaterialFunctionCall(
    BaseNormal=@n20.Result
    AdditionalNormal=MaterialFunctionCall(Normal=@n35.RGB, Flatness=@n36) {materialFunction=&FlattenNormal}.Result
   ) {materialFunction=&BlendAngleCorrectedNormals}.Result
  ) {attributeSetTypes [
    #1
    #2
    #3
    #4
   ]}
  NamedRerouteDeclaration(None=@n54) {name=JuiceMatAttributes, nodeColor=<rgba 0 0.9766 1 1>}
  n55 = TextureSample(UVs=Divide(A=@n25, B=Constant() {r=5599})) {
   samplerSource=SSM_Wrap_WorldGroupSettings
   samplerType=SAMPLERTYPE_Grayscale
   texture=&T_ApolloGrass_ColorMask
  }
  Saturate(None=LinearInterpolate(A=Constant(), B=Constant() {r=1.5}, Alpha=@n55.A))
  ScalarParameter() {defaultValue=-3600, group=Grass, parameterName=GrassColorMaskC Scale}
 ]}
