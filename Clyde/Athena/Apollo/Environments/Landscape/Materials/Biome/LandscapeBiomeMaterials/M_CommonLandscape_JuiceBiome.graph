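# Apollo landscape material for the Juice biome, blending painted ground layers, Forest and JuiceGrass biomes, slope cliffs and procedural snow.
# The result writes to and samples a runtime virtual texture, with a simplified mobile layer path.
# Grass output derives foliage from layer and height masks, limiting JuiceGrass by a Slurp sphere mask.
# JuiceGrass requires an overlapping BGA_Clyde_JuiceWater volume; SlurpActive gates deferred-path sparkle emissive on sand.
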Material /Clyde/Athena/Apollo/Environments/Landscape/Materials/Biome/LandscapeBiomeMaterials/M_CommonLandscape_JuiceBiome
graph {
 properties {
  bTangentSpaceNormal=false
  bUseMaterialAttributes=true
  bUsedWithNanite=true
  bUsedWithVirtualHeightfieldMesh=true
  physMaterial=/Game/FortressPhysicalMaterials/Grass
  pixelDepthOffsetMode=PDOM_Legacy
  usedShadingModels=Default Lit
 }
 nodes [
  Constant() {r=-4400}
  n1 = LandscapeLayerSample() {parameterName=Gravel}
  n2 = LandscapeLayerSample() {parameterName=Mud}
  n3 = Max(A=Multiply(A=Max(A=@n1, B=@n2), B=Constant() {r=2}), B=Multiply(A=LandscapeLayerSample() {parameterName=Sand}, B=Constant() {r=5}))
  n4 = Ceil(None=LandscapeLayerSample() {parameterName=Asphalt})
  n5 = MaterialFunctionCall!(
   UseBiomeColorMask_1
   BiomeMask_1
   GrassBorderColorMask_1=@n11.MaskForGrassBorderColor
   HeightmapFromCliff_1=@n11.HeightmapForGrass
   Grass-EdgeBiomeColorB_1
   Grass-EdgeBiomeColorA_1
   GravelBiomeColorDiv_1
   GravelBiomeColor_1
   MudBiomeColorDiv_1
   MudBiomeColor_1
   MudBiomeRoughMod_1
   SandBiomeColorDiv_1
   SandBiomeColor_1
  ) {materialFunction=/Clyde/Athena/Apollo/Environments/Landscape/Materials/Biome/CommonMaterials/MF_CommonLandscapeLayers}
  n6 = MaterialFunctionCall!(
   Material 1=@n5.Material
   Height Texture 1=@n5.Heightmap
   Material 2=@n5.Material
   Height Texture 2=@n5.Heightmap
   Transistion Phase=LandscapeLayerSample() {parameterName=NoFoliage}
   UseContrast
   Contrast
  ) {materialFunction=/Clyde/Athena/Apollo/Environments/Landscape/Materials/Biome/CommonMaterials/MF_BlendLandscapeLayers}
  n7 = MaterialFunctionCall!(UseBiomeColorMask, BiomeColorMask, BiomeColorDiv, BiomeColor) {materialFunction=/Clyde/Athena/Apollo/Environments/Landscape/Materials/Biome/UniqueBiomeMaterialFunctions/MF_Biome_Forest}
  n8 = MaterialFunctionCall(
   Material 1=@n6.BlendedMaterial
   Height Texture 1=@n6.BlendedMaterial
   Material 2=@n7.Material
   Height Texture 2=@n7.Material
   Transistion Phase=LandscapeLayerSample() {parameterName=BiomeB}
  ) {materialFunction=&MF_BlendLandscapeLayers}
  n9 = MaterialFunctionCall!(GrassBorderColorMask_1, UseBiomeColorMask, In=@n8.BlendedMaterial, BiomeColorMask, BiomeColorDiv, BiomeColor) {
   desc="Requires BGA_Clyde_JuiceWater actor to be in the world with a volume overlapping the LS for this biome to show up. "
   materialFunction=&/MF_JuiceGrass
  }
  n10 = MaterialFunctionCall!(
   Material 1=@n8.BlendedMaterial
   Height Texture 1=@n8.BlendedMaterial
   Material 2=@n9.Material
   Height Texture 2=@n9.Material
   Transistion Phase=LandscapeLayerSample() {parameterName=BiomeC}
   UseContrast
   Contrast
   AddLayerMask
  ) {materialFunction=/Game/Athena/Apollo/Environments/Landscape/Materials/Biome/CommonMaterials/MF_BlendLandscapeLayers}
  n11 = MaterialFunctionCall!(BiomeColorMask, Material=@n10.BlendedMaterial, Heightmap=@n10.BlendedMaterial, UseBiomeColorMask, BiomeColorDiv, BiomeColor) {materialFunction=/Clyde/Athena/Apollo/Environments/Landscape/Materials/Biome/CommonMaterials/MF_SlopeBlendCliff}
  n12 = Max(A=@n11.LandscapeGrassMeshMask, B=Saturate(None=Max(A=@n3, B=@n4)))
  n13 = MaterialFunctionCall!(Material=@n11.Material) {materialFunction=/Clyde/Athena/Apollo/Maps/Landscape/VT/MaterialFunctions/TransformTangentToWorld}
  n14 = MaterialFunctionCall!(MaterialIn=@n13.Result) {materialFunction=/Clyde/Athena/Apollo/Environments/Landscape/Materials/Biome/UniqueBiomeMaterialFunctions/MF_ProceduralSnow}
  MaterialExpressionSubtract_9 = Subtract(A=Subtract(A=OneMinus(None=@n12), B=@n14.MaterialMask), B=LandscapeLayerSample() {parameterName=NoFoliage})
  MaterialExpressionConstant_10 = Constant()
  MaterialExpressionSubtract_6 = Subtract(A=Constant())
  n15 = Constant()
  n16 = ComponentMask(None=WorldPosition().XYZ) {b=true}
  n17 = Subtract(A=@n16, B=Constant() {r=-5100})
  n18 = Constant() {r=1}
  n19 = OneMinus(None=Saturate(None=Divide(A=@n17, B=@n18)))
  n20 = Subtract(A=Subtract(A=LandscapeLayerSample() {parameterName=Sand}, B=@n11.LandscapeGrassMeshMask), B=@n19)
  MaterialExpressionSubtract_7 = Subtract(A=ShadingPathSwitch(Default=@n20, Deferred=@n20, Forward=@n15, Mobile=@n15), B=LandscapeLayerSample() {parameterName=NoFoliage})
  n21 = Constant()
  n22 = Divide(A=Subtract(A=@n16, B=Constant() {r=-4600}), B=@n18)
  n23 = Subtract(A=OneMinus(None=Saturate(None=@n22)), B=@n19)
  MaterialExpressionReroute_11 = Reroute(None=Subtract(A=ShadingPathSwitch(Default=@n23, Deferred=@n23, Forward=@n21, Mobile=@n21), B=LandscapeLayerSample() {parameterName=NoFoliage}))
  MaterialExpressionNamedRerouteUsage_0 = NamedRerouteUsage()
  LandscapeGrassOutput!(
   Grass_Standard=@MaterialExpressionSubtract_9
   Bush=@MaterialExpressionSubtract_9
   Scree=@MaterialExpressionConstant_10
   Scree_Grass=@MaterialExpressionSubtract_6
   Heather=@MaterialExpressionSubtract_6
   SwampGrass=@MaterialExpressionSubtract_6
   SwampReeds=@MaterialExpressionSubtract_6
   Forest_Grass=@MaterialExpressionSubtract_6
   Beach=@MaterialExpressionSubtract_7
   Ocean=@MaterialExpressionReroute_11
   JuiceGrass=@MaterialExpressionNamedRerouteUsage_0
  ) {grassTypes [
    Grass_Standard {
     grassType=/Clyde/Landscape/GrassTypes/LandscapeGrassType_StandardGrass
     input {
      expression=&self:MaterialExpressionSubtract_9
      inputName=None
      mask=0
      maskA=0
      maskB=0
      maskG=0
      maskR=0
      outputIndex=0
     }
    }
    Bush {
     grassType=/Clyde/Landscape/GrassTypes/Athena_Foliage_Bushes
     input {
      expression=&self:MaterialExpressionSubtract_9
      inputName=None
      mask=0
      maskA=0
      maskB=0
      maskG=0
      maskR=0
      outputIndex=0
     }
    }
    Scree {
     grassType=/Clyde/Landscape/GrassTypes/LandscapeGrassType_Scree
     input {
      expression=&self:MaterialExpressionConstant_10
      inputName=None
      mask=0
      maskA=0
      maskB=0
      maskG=0
      maskR=0
      outputIndex=0
     }
    }
    Scree_Grass {
     grassType=/Clyde/Landscape/GrassTypes/LandscapeGrassType_ScreeGrass
     input {
      expression=&self:MaterialExpressionSubtract_6
      inputName=None
      mask=0
      maskA=0
      maskB=0
      maskG=0
      maskR=0
      outputIndex=0
     }
    }
    Heather {
     grassType=/Clyde/Landscape/GrassTypes/LandscapeGrassType_Heather
     input {
      expression=&self:MaterialExpressionSubtract_6
      inputName=None
      mask=0
      maskA=0
      maskB=0
      maskG=0
      maskR=0
      outputIndex=0
     }
    }
    SwampGrass {
     grassType=/Clyde/Landscape/GrassTypes/LandscapeGrassType_SwampGrass
     input {
      expression=&self:MaterialExpressionSubtract_6
      inputName=None
      mask=0
      maskA=0
      maskB=0
      maskG=0
      maskR=0
      outputIndex=0
     }
    }
    SwampReeds {
     grassType=/Clyde/Landscape/GrassTypes/LandscapeGrassType_SwampReeds
     input {
      expression=&self:MaterialExpressionSubtract_6
      inputName=None
      mask=0
      maskA=0
      maskB=0
      maskG=0
      maskR=0
      outputIndex=0
     }
    }
    Forest_Grass {
     grassType=/Clyde/Landscape/GrassTypes/LandscapeGrassType_ForestGrass
     input {
      expression=&self:MaterialExpressionSubtract_6
      inputName=None
      mask=0
      maskA=0
      maskB=0
      maskG=0
      maskR=0
      outputIndex=0
     }
    }
    Beach {
     grassType=/Clyde/Landscape/GrassTypes/LandscapeGrassType_Beach
     input {
      expression=&self:MaterialExpressionSubtract_7
      inputName=None
      mask=0
      maskA=0
      maskB=0
      maskG=0
      maskR=0
      outputIndex=0
     }
    }
    Ocean {
     grassType=/Clyde/Landscape/GrassTypes/LandscapeGrassType_Ocean
     input {
      expression=&self:MaterialExpressionReroute_11
      inputName=None
      mask=0
      maskA=0
      maskB=0
      maskG=0
      maskR=0
      outputIndex=0
     }
    }
    JuiceGrass {
     grassType=/Clyde/Landscape/GrassTypes/LandscapeGrassType_JuiceGrass
     input {
      expression=&self:MaterialExpressionNamedRerouteUsage_0
      inputName=None
      mask=0
      maskA=0
      maskB=0
      maskG=0
      maskR=0
      outputIndex=0
     }
    }
   ]}
  n24 = MaterialFunctionCall!(UseSnow=StaticBool() {value=true}) {materialFunction=/Clyde/Athena/Apollo/Environments/Landscape/Materials/Biome/CommonMaterials/Mobile/MF_Mobile_CommonLandscapeLayers}
  n25 = MaterialFunctionCall(In=@n24.Material) {
   desc="Requires BGA_Clyde_JuiceWater actor to be in the world with a volume overlapping the LS for this biome to show up. "
   materialFunction=&MF_JuiceGrass
  }
  n26 = GetMaterialAttributes(None=MaterialFunctionCall(
    Material 1=@n24.Material
    Material 2=@n25.Mobile-Material
    Height Texture 2=@n25.Mobile-Material
    Transistion Phase=LandscapeLayerSample() {parameterName=BiomeC}
   ) {materialFunction=&MF_BlendLandscapeLayers}.BlendedMaterial) {attributeGetTypes [#1]}
  MaterialExpressionConstant_13 = Constant()
  n27 = LandscapeLayerBlend!(
   Layer Gravel=@MaterialExpressionConstant_13
   Layer Mud=@MaterialExpressionConstant_13
   Layer Sand=@MaterialExpressionConstant_13
   Layer Asphalt=@MaterialExpressionConstant_13
   Layer NoFoliage=@MaterialExpressionConstant_13
   Layer JuiceWater=@MaterialExpressionConstant_13
  ) {layers [
    {
     blendType=LB_WeightBlend
     constHeightInput=0
     layerName=Gravel
     previewWeight=0
     constLayerInput=<0 0 0>
     heightInput {expression=null, inputName=None, mask=0, maskA=0, maskB=0, maskG=0, maskR=0, outputIndex=0}
     layerInput {
      expression=&self:MaterialExpressionConstant_13
      inputName=None
      mask=0
      maskA=0
      maskB=0
      maskG=0
      maskR=0
      outputIndex=0
     }
    }
    {
     blendType=LB_WeightBlend
     constHeightInput=0
     layerName=Mud
     previewWeight=0
     constLayerInput=<0 0 0>
     heightInput {expression=null, inputName=None, mask=0, maskA=0, maskB=0, maskG=0, maskR=0, outputIndex=0}
     layerInput {
      expression=&self:MaterialExpressionConstant_13
      inputName=None
      mask=0
      maskA=0
      maskB=0
      maskG=0
      maskR=0
      outputIndex=0
     }
    }
    {
     blendType=LB_WeightBlend
     constHeightInput=0
     layerName=Sand
     previewWeight=0
     constLayerInput=<0 0 0>
     heightInput {expression=null, inputName=None, mask=0, maskA=0, maskB=0, maskG=0, maskR=0, outputIndex=0}
     layerInput {
      expression=&self:MaterialExpressionConstant_13
      inputName=None
      mask=0
      maskA=0
      maskB=0
      maskG=0
      maskR=0
      outputIndex=0
     }
    }
    {
     blendType=LB_WeightBlend
     constHeightInput=0
     layerName=Asphalt
     previewWeight=0
     constLayerInput=<0 0 0>
     heightInput {expression=null, inputName=None, mask=0, maskA=0, maskB=0, maskG=0, maskR=0, outputIndex=0}
     layerInput {
      expression=&self:MaterialExpressionConstant_13
      inputName=None
      mask=0
      maskA=0
      maskB=0
      maskG=0
      maskR=0
      outputIndex=0
     }
    }
    {
     blendType=LB_WeightBlend
     constHeightInput=0
     layerName=NoFoliage
     previewWeight=0
     constLayerInput=<0 0 0>
     heightInput {expression=null, inputName=None, mask=0, maskA=0, maskB=0, maskG=0, maskR=0, outputIndex=0}
     layerInput {
      expression=&self:MaterialExpressionConstant_13
      inputName=None
      mask=0
      maskA=0
      maskB=0
      maskG=0
      maskR=0
      outputIndex=0
     }
    }
    {
     blendType=LB_WeightBlend
     constHeightInput=0
     layerName=JuiceWater
     previewWeight=0
     constLayerInput=<0 0 0>
     heightInput {expression=null, inputName=None, mask=0, maskA=0, maskB=0, maskG=0, maskR=0, outputIndex=0}
     layerInput {
      expression=&self:MaterialExpressionConstant_13
      inputName=None
      mask=0
      maskA=0
      maskB=0
      maskG=0
      maskR=0
      outputIndex=0
     }
    }
   ]}
  n28 = SetMaterialAttributes!(MaterialAttributes=@n26.MaterialAttributes, Base Color=Add(A=@n26.Base Color, B=Multiply(A=@n27) {constB=0})) {attributeSetTypes [#1]}
  n29 = RuntimeVirtualTextureSampleParameter() {
   group=RVT
   materialType=BaseColor_Normal_Specular_YCoCg
   parameterName=RVT Sample Target
   virtualTexture=/Game/Athena/Apollo/RVT/RVT_Apollo_Landscape
  }
  n30 = SetMaterialAttributes!(MaterialAttributes, Base Color=@n29.BaseColor, Specular=@n29.BaseColor, Roughness=@n29.BaseColor, Normal=@n29.BaseColor) {attributeSetTypes [
    #1
    #2
    #3
    #4
   ]}
  n31 = BlendMaterialAttributes(A=@n13.Result, B=@n14.Material, Alpha=@n14.Material) {desc="Vertex blend type altered!", vertexAttributeBlendType=UseA}
  n32 = VirtualTextureFeatureSwitch!(
   No=@n28
   Yes=StaticSwitchParameter(
    True=MaterialFunctionCall(NoRVT=@n31, RVT=@n30) {materialFunction=&FN_RVT_Switch_Material}.Result
    False=@n31
   ) {defaultValue=true, group=RVT Base, parameterName=RVT Enable}
  )
  n33 = ShadingPathSwitch(Default=@n32, Mobile=QualitySwitch(Default=@n28, Medium=@n32))
  n34 = Multiply(
   A=LandscapeLayerSample() {parameterName=Sand}
   B=CollectionParameter() {collection=&FortniteMaterialParameters, parameterName=SlurpActive}
  )
  n35 = MaterialFunctionCall(
   UVs=WorldPosition().XY
   CenterPosition=Constant2Vector() {g=-108888.8672, r=-33334.2461}
   Radius=ScalarParameter() {group=PostRVTSnow, parameterName=Beach Glitter Radius}
   Density=ScalarParameter() {group=PostRVTSnow, parameterName=Beach Glitter Density}
  ) {materialFunction=&RadialGradientExponential}
  n36 = Constant() {r=1}
  n37 = ComponentMask(None=NamedRerouteUsage()) {g=true}
  n38 = BlendMaterialAttributes(
   A=@n33
   B=SetMaterialAttributes!(MaterialAttributes=@n33, Emissive Color=NamedRerouteUsage()) {attributeSetTypes [#5]}
   Alpha=Multiply(
    A=Multiply(A=@n34, B=@n35.RadialGradientExponential)
    B=MaterialFunctionCall!(NoRVT=@n36, RVT=@n37) {materialFunction=/Game/Packages/Fortress_SharedMaterials/Base_Material_Functions/FN_RVT_Switch_Scalar}.Result
   )
  )
  MaterialFunctionCall!(Material=ShadingPathSwitch(Default=@n38, Deferred=@n38, Forward=@n33, Mobile=@n33), DebugAsBasecolor, DebugBiomeColor) {materialFunction=/Clyde/Athena/Apollo/Environments/Landscape/Materials/Biome/CommonMaterials/MF_LS_DEBUG}
  n39 = ScalarParameter() {defaultValue=20, group=PostRVTSnow, parameterName=SparkleMaskExponent}
  n40 = Constant() {r=1}
  Multiply(
   A=@n39
   B=MaterialFunctionCall(Value_Morning=@n40, Value_Day=Constant() {r=2}, Value_Evening=@n40, Value_Night=@n40) {materialFunction=&Scalar_Time_of_Day}.Result
  )
  n41 = SkyLightEnvMapSample(Direction=ReflectionVectorWS(CustomWorldNormal=NamedRerouteUsage()))
  n42 = NamedRerouteUsage()
  n43 = Constant() {r=850}
  n44 = TextureSample(UVs=Panner(Coordinate=MaterialFunctionCall!(Scale=@n43) {materialFunction=/Game/Athena/Apollo/Environments/Landscape/Materials/Biome/CommonMaterials/MF_SimpleWorldUV}.Result) {speedX=0.1}) {
   samplerSource=SSM_Wrap_WorldGroupSettings
   samplerType=SAMPLERTYPE_Grayscale
   texture=/Game/TimeOfDay/TODM_Sequence/Materials/Sky/Textures/LowResBlurredNoise_Radish_LF_Root
  }
  n45 = LinearInterpolate(Alpha=@n44.R) {constA=-0.3, constB=0.2}
  n46 = ScalarParameter() {defaultValue=2.5, group=PostRVTSnow, parameterName=Morning/Evening-Multi}
  n47 = Multiply(
   A=VectorParameter() {parameterName=Sparkle_Color, defaultValue=<rgba 15 0 14.431 1>}.RGB
   B=MaterialFunctionCall(
    Value_Morning=@n46
    Value_Day=ScalarParameter() {defaultValue=8, group=PostRVTSnow, parameterName=Day_Multi}
    Value_Evening=@n46
    Value_Night=ScalarParameter() {defaultValue=1, group=PostRVTSnow, parameterName=Night_Emissive_Multi}
   ) {materialFunction=&Scalar_Time_of_Day}.Result
  )
  n48 = LandscapeLayerSample() {parameterName=Sand}
  n49 = Saturate(None=Multiply(A=Add(A=@n48) {constB=-0.3}) {constB=2})
  n50 = Multiply(
   A=Multiply(
    A=Multiply(A=@n41, B=@n42)
    B=MaterialFunctionCall(Hue Shift Percentage=@n45, Texture=@n47) {materialFunction=&HueShift}.Result
   )
   B=NamedRerouteDeclaration(None=Power(Base=@n49) {constExponent=1}) {name=SnowMaskModified, nodeColor=<rgba 0.125 0 1 1>}
  )
  n51 = ScalarParameter() {defaultValue=1000, group=PostRVTSnow, parameterName=Sparkle_Mask_Size}
  n52 = TextureSample(UVs=Panner(Coordinate=MaterialFunctionCall(Scale=@n51) {materialFunction=&MF_SimpleWorldUV}.Result) {speedX=0.2, speedY=0.1}) {
   samplerSource=SSM_Wrap_WorldGroupSettings
   samplerType=SAMPLERTYPE_Grayscale
   texture=&LowResBlurredNoise_Radish_LF_Root
  }
  Multiply(
   A=@n50
   B=MaterialFunctionCall!(Default=LinearInterpolate(Alpha=@n52.R) {constA=0.05}, Low/Mobile=Constant() {r=0.8}) {materialFunction=/Game/Effects/Fort_Effects/Materials/_Functions/MF_LowMobile_PathingSwitch}.Result
  )
  n53 = ComponentMask(None=CollectionParameter() {collection=&FortniteMaterialParameters, parameterName=SlurpyCenter}) {b=true, g=true, r=true}
  n54 = ComponentMask(None=CollectionParameter() {collection=&FortniteMaterialParameters, parameterName=SlurpyRadius}) {b=true, g=true, r=true}
  n55 = WorldPosition()
  n56 = ScalarParameter() {defaultValue=1, parameterName=SlurpyGradient}
  n57 = Subtract(
   A=Multiply(A=LandscapeLayerSample() {parameterName=BiomeC}, B=SphereMask(A=@n55.XYZ, B=@n53, Radius=@n54, Hardness=@n56))
   B=LandscapeLayerSample() {parameterName=NoFoliage}
  )
  n58 = Max(A=LandscapeLayerSample() {parameterName=Mud}, B=LandscapeLayerSample() {parameterName=Gravel})
  n59 = LandscapeLayerSample() {parameterName=Sand}
  n60 = Multiply(A=Max(A=Max(A=@n58, B=@n59), B=LandscapeLayerSample() {parameterName=Asphalt}), B=Constant() {r=2})
  NamedRerouteDeclaration(None=Saturate(None=Subtract(A=@n57, B=@n60))) {name=JuiceGrassFoliage, nodeColor=<rgba 0 0.5547 1 1>}
  NamedRerouteDeclaration(None=@n29.Normal) {name=RVT_Mask, nodeColor=<rgba 1 0.2344 0 1>}
  n61 = NamedRerouteUsage()
  n62 = TextureSample(UVs=@n61) {
   samplerSource=SSM_Wrap_WorldGroupSettings
   samplerType=SAMPLERTYPE_Normal
   texture=/Game/Environments/Asteria/Terrain/Textures/T_SnowSparkleCircles_N
  }
  n63 = ComponentMask(None=@n62.RGB) {b=true, g=true, r=true}
  n64 = Constant3Vector() {constant=<rgba 0 0 -0.9 1>}
  n65 = Transform(None=Normalize(VectorInput=Add(A=@n63, B=@n64)))
  n66 = DotProduct(A=ReflectionVectorWS(CustomWorldNormal=@n65), B=CameraVectorWS())
  n67 = TextureSample(UVs=@n61) {
   samplerSource=SSM_Wrap_WorldGroupSettings
   samplerType=SAMPLERTYPE_LinearColor
   texture=/Game/Environments/Asteria/Terrain/Textures/T_SnowSparkleCircles_M
  }
  n68 = ComponentMask(None=Saturate(None=Multiply(A=@n67.RGB) {constB=3})) {r=true}
  NamedRerouteDeclaration(None=Power(Base=Multiply(A=@n66, B=@n68), Exp=@n39)) {name=HighendSparkleMask, nodeColor=<rgba 0 0.1094 1 1>}
  NamedRerouteDeclaration(None=@n10.LayerMask) {name=JuiceBiomeMask, nodeColor=<rgba 1 0 0.7266 1>}
  NamedRerouteDeclaration(None=@n29.BaseColor) {name=rvt_bLUE, nodeColor=<rgba 0.3125 1 0 1>}
  NamedRerouteDeclaration(None=@n65) {name=sparklenormal, nodeColor=<rgba 1 0.1875 0 1>}
  NamedRerouteDeclaration(None=MaterialFunctionCall(Default=@n50, Low/Mobile=Constant()) {materialFunction=&MF_LowMobile_PathingSwitch}.Result) {name=SparkleEmissive, nodeColor=<rgba 0 0.2734 1 1>}
  n69 = TextureCoordinate()
  n70 = ScalarParameter() {defaultValue=4, parameterName=SnowSparkleSize}
  n71 = ScalarParameter() {defaultValue=200, group=PostRVTSnow, parameterName=SparkleSize}
  NamedRerouteDeclaration(None=StaticSwitchParameter(
    True=Multiply(A=@n69, B=@n70)
    False=MaterialFunctionCall(Scale=@n71) {materialFunction=&MF_SimpleWorldUV}.Result
   ) {parameterName=UseOnMesh}) {name=SparkleCoords, nodeColor=<rgba 1 0 0.2812 1>}
  n72 = Constant()
  QualitySwitch(Default=@n72, Low=@n72, High=@n72, Medium=@n72, Epic=@n50)
  n73 = GetMaterialAttributes(None=@n31) {attributeGetTypes [
    #1
    #2
    #3
    #4
   ]}
  RuntimeVirtualTextureOutput!(
   BaseColor=MaterialFunctionCall!(BaseColor=@n73.Base Color) {materialFunction=/Game/Athena/Apollo/Environments/Landscape/Materials/Biome/CommonMaterials/MF_ReduceRTV-CompressionArtifacts}.BaseColor
   Specular=@n73.Specular
   Roughness=@n73.Specular
   Normal=@n73.Specular
   WorldHeight=Add(A=BreakMaterialAttributes!(Attr=@n73.MaterialAttributes).Displacement, B=WorldPosition().Z)
   Opacity
   Mask
   Displacement
   Mask4
  )
  ScalarParameter() {defaultValue=8, group=PostRVTSnow, parameterName=SparkleEmissiveBrightness}
  TextureCoordinate() {uTiling=18, vTiling=18}
  VectorParameter() {parameterName=DebugBiomeColor, defaultValue=<rgba 0.0445 1 0 1>}
 ]
}
