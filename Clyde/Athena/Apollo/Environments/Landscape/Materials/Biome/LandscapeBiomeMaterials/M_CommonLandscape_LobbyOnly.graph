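# Nanite-ready landscape material for the Apollo lobby island, blending common layers, Forest and FarmGrass biomes via painted weights, then slope-based cliffs.
# Deferred shading writes the result to a runtime virtual texture and samples it back through a switch.
# Forward/mobile uses simplified layers plus a zero-multiplied layer blend.
# Grass output spawns standard grass, farm flowers, beach and height-band ocean types from layer and slope masks, minus NoFoliage.
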Material /Clyde/Athena/Apollo/Environments/Landscape/Materials/Biome/LandscapeBiomeMaterials/M_CommonLandscape_LobbyOnly
graph {
 properties {
  bTangentSpaceNormal=false
  bUseMaterialAttributes=true
  bUsedWithNanite=true
  pixelDepthOffsetMode=PDOM_Legacy
  usedShadingModels=Default Lit
 }
 nodes [
  n1 = Multiply(A=Max(A=LandscapeLayerSample() {parameterName=Gravel}, B=LandscapeLayerSample() {parameterName=Mud}), B=Constant() {r=2})
  n2 = Multiply(A=LandscapeLayerSample() {parameterName=Sand}, B=Constant() {r=5})
  n3 = LandscapeLayerSample() {parameterName=Asphalt}
  n4 = Saturate(None=Max(A=Max(A=@n1, B=@n2), B=Ceil(None=@n3)))
  n5 = MaterialFunctionCall!(
   UseBiomeColorMask_1
   BiomeMask_1
   GrassBorderColorMask_1=@n11.MaskForGrassBorderColor
   HeightmapFromCliff_1=@n11.HeightmapForGrass
   Grass-EdgeBiomeColorB_1
   Grass-EdgeBiomeColorA_1
   GravelBiomeColorDiv_1
   GravelBiomeColor_1
   MudBiomeColorDiv_1
   MudBiomeColor_1
   MudBiomeRoughMod_1
   SandBiomeColorDiv_1
   SandBiomeColor_1
  ) {materialFunction=/Clyde/Athena/Apollo/Environments/Landscape/Materials/Biome/CommonMaterials/MF_CommonLandscapeLayers}
  n6 = MaterialFunctionCall!(
   Material 1=@n5.Material
   Height Texture 1=@n5.Heightmap
   Material 2=@n5.Material
   Height Texture 2=@n5.Heightmap
   Transistion Phase=LandscapeLayerSample() {parameterName=NoFoliage}
   UseContrast
   Contrast
  ) {materialFunction=/Clyde/Athena/Apollo/Environments/Landscape/Materials/Biome/CommonMaterials/MF_BlendLandscapeLayers}
  n7 = MaterialFunctionCall!(UseBiomeColorMask, BiomeColorMask, BiomeColorDiv, BiomeColor) {materialFunction=/Clyde/Athena/Apollo/Environments/Landscape/Materials/Biome/UniqueBiomeMaterialFunctions/MF_Biome_Forest}
  n8 = MaterialFunctionCall(
   Material 1=@n6.BlendedMaterial
   Height Texture 1=@n6.BlendedMaterial
   Material 2=@n7.Material
   Height Texture 2=@n7.Material
   Transistion Phase=LandscapeLayerSample() {parameterName=BiomeB}
  ) {materialFunction=&MF_BlendLandscapeLayers}
  n9 = MaterialFunctionCall!(
   GrassHeightmap=@n5.GrassNoiseMask_1
   GrassNoiseMask=@n5.GrassHeightmap_1
   GrassBaseColorDetail=@n5.GrassBaseColorDetail_1
   GrassMaterial=@n5.Material
  ) {materialFunction=/Clyde/Athena/Apollo/Environments/Landscape/Materials/Biome/UniqueBiomeMaterialFunctions/MF_Biome_FarmGrass}
  n10 = MaterialFunctionCall(
   Material 1=@n8.BlendedMaterial
   Height Texture 1=@n8.BlendedMaterial
   Material 2=@n9.Material
   Height Texture 2=@n9.Material
   Transistion Phase=LandscapeLayerSample() {parameterName=BiomeC}
  ) {materialFunction=&MF_BlendLandscapeLayers}
  n11 = MaterialFunctionCall!(BiomeColorMask, Material=@n10.BlendedMaterial, Heightmap=@n10.BlendedMaterial, UseBiomeColorMask, BiomeColorDiv, BiomeColor) {materialFunction=/Clyde/Athena/Apollo/Environments/Landscape/Materials/Biome/CommonMaterials/MF_SlopeBlendCliff}
  MaterialExpressionSubtract_3 = Subtract(A=OneMinus(None=Max(A=@n11.LandscapeGrassMeshMask, B=@n4)), B=LandscapeLayerSample() {parameterName=NoFoliage})
  n12 = Constant() {r=2}
  n13 = LandscapeLayerSample() {parameterName=BiomeC}
  MaterialExpressionSubtract_0 = Subtract(A=Add(A=Constant() {r=-0.5}, B=Multiply(A=@n12, B=@n13)), B=LandscapeLayerSample() {parameterName=NoFoliage})
  n14 = Constant()
  n15 = Subtract(A=LandscapeLayerSample() {parameterName=Sand}, B=@n11.LandscapeGrassMeshMask)
  MaterialExpressionSubtract_1 = Subtract(A=ShadingPathSwitch(Default=@n15, Deferred=@n15, Forward=@n14, Mobile=@n14), B=LandscapeLayerSample() {parameterName=NoFoliage})
  n16 = Constant()
  n17 = ComponentMask(None=WorldPosition().XYZ) {b=true}
  n18 = Constant() {r=-4400}
  n19 = Constant() {r=1}
  n20 = Saturate(None=Divide(A=Subtract(A=@n17, B=@n18), B=@n19))
  n21 = Constant() {r=-5100}
  n22 = Saturate(None=Divide(A=Subtract(A=@n17, B=@n21), B=@n19))
  n23 = Subtract(A=Subtract(A=OneMinus(None=@n20), B=OneMinus(None=@n22)), B=@n11.LandscapeGrassMeshMask)
  MaterialExpressionSubtract_2 = Subtract(A=ShadingPathSwitch(Default=@n23, Deferred=@n23, Forward=@n16, Mobile=@n16), B=LandscapeLayerSample() {parameterName=NoFoliage})
  LandscapeGrassOutput!(
   Grass_Standard=@MaterialExpressionSubtract_3
   FarmFlowers=@MaterialExpressionSubtract_0
   Beach=@MaterialExpressionSubtract_1
   Ocean=@MaterialExpressionSubtract_2
  ) {grassTypes [
    Grass_Standard {
     grassType=/Clyde/Landscape/GrassTypes/LandscapeGrassType_StandardGrass_Lobby
     input {
      expression=&self:MaterialExpressionSubtract_3
      inputName=None
      mask=0
      maskA=0
      maskB=0
      maskG=0
      maskR=0
      outputIndex=0
     }
    }
    FarmFlowers {
     grassType=/Clyde/Landscape/GrassTypes/LandscapeGrassType_FarmFlowers
     input {
      expression=&self:MaterialExpressionSubtract_0
      inputName=None
      mask=0
      maskA=0
      maskB=0
      maskG=0
      maskR=0
      outputIndex=0
     }
    }
    Beach {
     grassType=/Clyde/Landscape/GrassTypes/LandscapeGrassType_Beach
     input {
      expression=&self:MaterialExpressionSubtract_1
      inputName=None
      mask=0
      maskA=0
      maskB=0
      maskG=0
      maskR=0
      outputIndex=0
     }
    }
    Ocean {
     grassType=/Clyde/Landscape/GrassTypes/LandscapeGrassType_Ocean
     input {
      expression=&self:MaterialExpressionSubtract_2
      inputName=None
      mask=0
      maskA=0
      maskB=0
      maskG=0
      maskR=0
      outputIndex=0
     }
    }
   ]}
  n24 = GetMaterialAttributes(None=MaterialFunctionCall!(UseSnow=StaticBool()) {materialFunction=/Clyde/Athena/Apollo/Environments/Landscape/Materials/Biome/CommonMaterials/Mobile/MF_Mobile_CommonLandscapeLayers}.Material) {attributeGetTypes [#1]}
  MaterialExpressionConstant_13 = Constant()
  n25 = LandscapeLayerBlend!(
   Layer Gravel=@MaterialExpressionConstant_13
   Layer Mud=@MaterialExpressionConstant_13
   Layer Sand=@MaterialExpressionConstant_13
   Layer Asphalt=@MaterialExpressionConstant_13
   Layer NoFoliage=@MaterialExpressionConstant_13
  ) {layers [
    {
     blendType=LB_WeightBlend
     constHeightInput=0
     layerName=Gravel
     previewWeight=0
     constLayerInput=<0 0 0>
     heightInput {expression=null, inputName=None, mask=0, maskA=0, maskB=0, maskG=0, maskR=0, outputIndex=0}
     layerInput {
      expression=&self:MaterialExpressionConstant_13
      inputName=None
      mask=0
      maskA=0
      maskB=0
      maskG=0
      maskR=0
      outputIndex=0
     }
    }
    {
     blendType=LB_WeightBlend
     constHeightInput=0
     layerName=Mud
     previewWeight=0
     constLayerInput=<0 0 0>
     heightInput {expression=null, inputName=None, mask=0, maskA=0, maskB=0, maskG=0, maskR=0, outputIndex=0}
     layerInput {
      expression=&self:MaterialExpressionConstant_13
      inputName=None
      mask=0
      maskA=0
      maskB=0
      maskG=0
      maskR=0
      outputIndex=0
     }
    }
    {
     blendType=LB_WeightBlend
     constHeightInput=0
     layerName=Sand
     previewWeight=0
     constLayerInput=<0 0 0>
     heightInput {expression=null, inputName=None, mask=0, maskA=0, maskB=0, maskG=0, maskR=0, outputIndex=0}
     layerInput {
      expression=&self:MaterialExpressionConstant_13
      inputName=None
      mask=0
      maskA=0
      maskB=0
      maskG=0
      maskR=0
      outputIndex=0
     }
    }
    {
     blendType=LB_WeightBlend
     constHeightInput=0
     layerName=Asphalt
     previewWeight=0
     constLayerInput=<0 0 0>
     heightInput {expression=null, inputName=None, mask=0, maskA=0, maskB=0, maskG=0, maskR=0, outputIndex=0}
     layerInput {
      expression=&self:MaterialExpressionConstant_13
      inputName=None
      mask=0
      maskA=0
      maskB=0
      maskG=0
      maskR=0
      outputIndex=0
     }
    }
    {
     blendType=LB_WeightBlend
     constHeightInput=0
     layerName=NoFoliage
     previewWeight=0
     constLayerInput=<0 0 0>
     heightInput {expression=null, inputName=None, mask=0, maskA=0, maskB=0, maskG=0, maskR=0, outputIndex=0}
     layerInput {
      expression=&self:MaterialExpressionConstant_13
      inputName=None
      mask=0
      maskA=0
      maskB=0
      maskG=0
      maskR=0
      outputIndex=0
     }
    }
   ]}
  n26 = SetMaterialAttributes!(MaterialAttributes=@n24.MaterialAttributes, Base Color=Add(A=@n24.Base Color, B=Multiply(A=@n25) {constB=0})) {attributeSetTypes [#1]}
  n27 = RuntimeVirtualTextureSampleParameter() {
   group=RVT
   materialType=BaseColor_Normal_Specular_YCoCg
   parameterName=RVT Sample Target
   virtualTexture=/Game/Athena/Apollo/RVT/RVT_Apollo_LobbyIsland
  }
  n28 = SetMaterialAttributes!(MaterialAttributes, Base Color=@n27.BaseColor, Specular=@n27.BaseColor, Roughness=@n27.BaseColor, Normal=@n27.BaseColor) {attributeSetTypes [
    #1
    #2
    #3
    #4
   ]}
  n29 = MaterialFunctionCall!(Material=@n11.Material) {materialFunction=/Clyde/Athena/Apollo/Maps/Landscape/VT/MaterialFunctions/TransformTangentToWorld}
  n30 = StaticSwitchParameter(
   True=MaterialFunctionCall!(RVT=@n28, ByPass RVT=@n29.Result) {materialFunction=/Game/Athena/Apollo/Maps/Landscape/VT/MaterialFunctions/EnableRVT}.Result
   False=@n29.Result
  ) {defaultValue=true, group=RVT Base, parameterName=RVT Enable}
  MaterialFunctionCall!(Material=ShadingPathSwitch(Default=@n30, Deferred=@n30, Forward=@n26, Mobile=@n26), DebugAsBasecolor, DebugBiomeColor) {materialFunction=/Clyde/Athena/Apollo/Environments/Landscape/Materials/Biome/CommonMaterials/MF_LS_DEBUG}
  n31 = GetMaterialAttributes(None=@n29.Result) {attributeGetTypes [
    #1
    #2
    #3
    #4
   ]}
  RuntimeVirtualTextureOutput!(
   BaseColor=MaterialFunctionCall!(BaseColor=@n31.Base Color) {materialFunction=/Game/Athena/Apollo/Environments/Landscape/Materials/Biome/CommonMaterials/MF_ReduceRTV-CompressionArtifacts}.BaseColor
   Specular=@n31.Specular
   Roughness=@n31.Specular
   Normal=@n31.Specular
   WorldHeight=ComponentMask(None=WorldPosition().XYZ) {b=true}
   Opacity
   Mask
   Displacement
   Mask4
  )
  VectorParameter() {parameterName=DebugBiomeColor, defaultValue=<rgba 0.0445 1 0 1>}
 ]
}
