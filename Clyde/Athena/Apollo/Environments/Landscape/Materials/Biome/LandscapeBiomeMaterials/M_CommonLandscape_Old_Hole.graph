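# Masked Apollo landscape material, named as the old hole variant, that shades biome terrain with a landscape visibility mask.
# It height-blends common layers with scree, forest and heather biomes, then adds slope cliffs and procedural snow.
# The result writes to a runtime virtual texture and can be sampled back, with simplified mobile-layer fallbacks through feature and quality switches.
# Layer, slope and snow masks drive landscape grass, while several grass types get constant-based inputs.
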Material /Clyde/Athena/Apollo/Environments/Landscape/Materials/Biome/LandscapeBiomeMaterials/M_CommonLandscape_Old_Hole
graph {
 properties {
  bTangentSpaceNormal=false
  bUsedWithNanite=true
  bUsedWithVirtualHeightfieldMesh=true
  blendMode=BLEND_Masked
  physMaterial=/Game/FortressPhysicalMaterials/Grass
  pixelDepthOffsetMode=PDOM_Legacy
  usedShadingModels=Default Lit
 }
 nodes [
  Constant3Vector() {constant=<rgba 1 0.7337 0.5061 1>}
  Constant3Vector() {constant=<rgba 0.42 0.5403 0.6 1>}
  Constant3Vector() {constant=<rgba 0.7 0.5136 0.3543 1>}
  Constant3Vector() {constant=<rgba 0.42 0.5403 0.6 1>}
  n1 = GetMaterialAttributes(None=MaterialFunctionCall!(UseSnow=StaticBool() {value=true}) {materialFunction=/Clyde/Athena/Apollo/Environments/Landscape/Materials/Biome/CommonMaterials/Mobile/MF_Mobile_CommonLandscapeLayers}.Material) {attributeGetTypes [#1]}
  MaterialExpressionConstant_16 = Constant()
  n2 = LandscapeLayerBlend!(
   Layer Gravel=@MaterialExpressionConstant_16
   Layer Mud=@MaterialExpressionConstant_16
   Layer Sand=@MaterialExpressionConstant_16
   Layer Asphalt=@MaterialExpressionConstant_16
   Layer NoFoliage=@MaterialExpressionConstant_16
  ) {layers [
    {
     blendType=LB_WeightBlend
     constHeightInput=0
     layerName=Gravel
     previewWeight=0
     constLayerInput=<0 0 0>
     heightInput {expression=null, inputName=None, mask=0, maskA=0, maskB=0, maskG=0, maskR=0, outputIndex=0}
     layerInput {
      expression=&self:MaterialExpressionConstant_16
      inputName=None
      mask=0
      maskA=0
      maskB=0
      maskG=0
      maskR=0
      outputIndex=0
     }
    }
    {
     blendType=LB_WeightBlend
     constHeightInput=0
     layerName=Mud
     previewWeight=0
     constLayerInput=<0 0 0>
     heightInput {expression=null, inputName=None, mask=0, maskA=0, maskB=0, maskG=0, maskR=0, outputIndex=0}
     layerInput {
      expression=&self:MaterialExpressionConstant_16
      inputName=None
      mask=0
      maskA=0
      maskB=0
      maskG=0
      maskR=0
      outputIndex=0
     }
    }
    {
     blendType=LB_WeightBlend
     constHeightInput=0
     layerName=Sand
     previewWeight=0
     constLayerInput=<0 0 0>
     heightInput {expression=null, inputName=None, mask=0, maskA=0, maskB=0, maskG=0, maskR=0, outputIndex=0}
     layerInput {
      expression=&self:MaterialExpressionConstant_16
      inputName=None
      mask=0
      maskA=0
      maskB=0
      maskG=0
      maskR=0
      outputIndex=0
     }
    }
    {
     blendType=LB_WeightBlend
     constHeightInput=0
     layerName=Asphalt
     previewWeight=0
     constLayerInput=<0 0 0>
     heightInput {expression=null, inputName=None, mask=0, maskA=0, maskB=0, maskG=0, maskR=0, outputIndex=0}
     layerInput {
      expression=&self:MaterialExpressionConstant_16
      inputName=None
      mask=0
      maskA=0
      maskB=0
      maskG=0
      maskR=0
      outputIndex=0
     }
    }
    {
     blendType=LB_WeightBlend
     constHeightInput=0
     layerName=NoFoliage
     previewWeight=0
     constLayerInput=<0 0 0>
     heightInput {expression=null, inputName=None, mask=0, maskA=0, maskB=0, maskG=0, maskR=0, outputIndex=0}
     layerInput {
      expression=&self:MaterialExpressionConstant_16
      inputName=None
      mask=0
      maskA=0
      maskB=0
      maskG=0
      maskR=0
      outputIndex=0
     }
    }
   ]}
  n3 = SetMaterialAttributes!(MaterialAttributes=@n1.MaterialAttributes, Base Color=Add(A=@n1.Base Color, B=Multiply(A=@n2) {constB=0})) {attributeSetTypes [#1]}
  n4 = RuntimeVirtualTextureSampleParameter() {
   group=RVT
   materialType=BaseColor_Normal_Specular_YCoCg
   parameterName=RVT Sample Target
   virtualTexture=/Game/Athena/Apollo/RVT/RVT_Apollo_Landscape
  }
  n5 = SetMaterialAttributes!(MaterialAttributes, Base Color=@n4.BaseColor, Specular=@n4.BaseColor, Roughness=@n4.BaseColor, Normal=@n4.BaseColor) {attributeSetTypes [
    #1
    #2
    #3
    #4
   ]}
  n6 = MaterialFunctionCall!(UseLandscapeCoords=StaticBool() {value=true}) {materialFunction=/Clyde/Athena/Apollo/Environments/Landscape/Materials/Biome/CommonMaterials/MF_ApolloTerrainMasks}
  n7 = LandscapeLayerSample() {parameterName=BiomeB}
  n8 = MaterialFunctionCall!(
   UseBiomeColorMask_1=StaticBool() {value=true}
   BiomeMask_1=@n6.Mountain
   GrassBorderColorMask_1=Max(A=@n7, B=@n17.MaskForGrassBorderColor)
   HeightmapFromCliff_1=@n17.HeightmapForGrass
   Grass-EdgeBiomeColorB_1=VectorParameter() {parameterName=GrassEdgeColorB, defaultValue=<rgba 0.135 0.14 0.007 1>}.RGB
   Grass-EdgeBiomeColorA_1=VectorParameter() {parameterName=GrassEdgeColorA, defaultValue=<rgba 0.0701 0.1 0.01 1>}.RGB
   GravelBiomeColorDiv_1=VectorParameter() {parameterName=Biome-GravelDiv, defaultValue=<rgba 1 0.7337 0.5061 1>}.RGB
   GravelBiomeColor_1=VectorParameter() {parameterName=Biome-GravelColor, defaultValue=<rgba 0.4275 0.4425 0.45 1>}.RGB
   MudBiomeColorDiv_1=VectorParameter() {parameterName=MountainBiomeMudDiv, defaultValue=<rgba 0.8 0.5412 0.32 1>}.R
   MudBiomeColor_1=VectorParameter() {parameterName=MountainBiomeMudColor, defaultValue=<rgba 0.35 0.4502 0.5 1>}.RGB
   MudBiomeRoughMod_1
   SandBiomeColorDiv_1=VectorParameter() {parameterName=BiomeSandDiv, defaultValue=<rgba 1 0.8306 0.6641 1>}.RGB
   SandBiomeColor_1=VectorParameter() {parameterName=BiomeSandColor, defaultValue=<rgba 0.2625 0.3354 0.35 1>}.RGB
  ) {materialFunction=/Clyde/Athena/Apollo/Environments/Landscape/Materials/Biome/CommonMaterials/MF_CommonLandscapeLayers}
  n9 = MaterialFunctionCall!(
   Material 1=@n8.Material
   Height Texture 1=@n8.Heightmap
   Material 2=@n8.Material
   Height Texture 2=@n8.Heightmap
   Transistion Phase=LandscapeLayerSample() {parameterName=NoFoliage}
   UseContrast
   Contrast
  ) {materialFunction=/Clyde/Athena/Apollo/Environments/Landscape/Materials/Biome/CommonMaterials/MF_BlendLandscapeLayers}
  n10 = LandscapeLayerSample() {parameterName=BiomeA}
  n11 = MaterialFunctionCall!(LayerWeight=@n10) {materialFunction=/Clyde/Athena/Apollo/Environments/Landscape/Materials/Biome/UniqueBiomeMaterialFunctions/MF_Scree}
  n12 = MaterialFunctionCall(
   Material 1=@n9.BlendedMaterial
   Height Texture 1=@n9.BlendedMaterial
   Material 2=@n11.Material
   Height Texture 2=@n11.Material
   Transistion Phase=@n10
   UseContrast=StaticBool() {value=true}
   Contrast=ScalarParameter() {defaultValue=0.5, parameterName=BiomeA-LayerContrast}
  ) {materialFunction=&MF_BlendLandscapeLayers}
  n13 = MaterialFunctionCall!(UseBiomeColorMask, BiomeColorMask, BiomeColorDiv, BiomeColor) {materialFunction=/Clyde/Athena/Apollo/Environments/Landscape/Materials/Biome/UniqueBiomeMaterialFunctions/MF_Biome_Forest}
  n14 = MaterialFunctionCall(
   Material 1=@n12.BlendedMaterial
   Height Texture 1=@n12.BlendedMaterial
   Material 2=@n13.Material
   Height Texture 2=@n13.Material
   Transistion Phase=@n7
  ) {materialFunction=&MF_BlendLandscapeLayers}
  n15 = MaterialFunctionCall!(GrassHeightmap=@n8.GrassHeightmap_1, GrassNoiseMask=@n8.GrassNoiseMask_1, GrassMaterial=@n8.Material) {materialFunction=/Clyde/Athena/Apollo/Environments/Landscape/Materials/Biome/UniqueBiomeMaterialFunctions/MF_Biome_Heather}
  n16 = MaterialFunctionCall(
   Material 1=@n14.BlendedMaterial
   Height Texture 1=@n14.BlendedMaterial
   Material 2=@n15.Material
   Height Texture 2=@n15.Material
   Transistion Phase=LandscapeLayerSample() {parameterName=BiomeC}
  ) {materialFunction=&MF_BlendLandscapeLayers}
  n17 = MaterialFunctionCall!(
   BiomeColorMask=@n6.Mountain
   Material=@n16.BlendedMaterial
   Heightmap=@n16.BlendedMaterial
   UseBiomeColorMask=StaticBool()
   BiomeColorDiv
   BiomeColor
  ) {materialFunction=/Clyde/Athena/Apollo/Environments/Landscape/Materials/Biome/CommonMaterials/MF_SlopeBlendCliff}
  n18 = MaterialFunctionCall!(Material=@n17.Material) {materialFunction=/Game/Athena/Apollo/Maps/Landscape/VT/MaterialFunctions/TransformTangentToWorld}
  n19 = MaterialFunctionCall!(MaterialIn=@n18.Result) {materialFunction=/Clyde/Athena/Apollo/Environments/Landscape/Materials/Biome/UniqueBiomeMaterialFunctions/MF_ProceduralSnow}
  n20 = BlendMaterialAttributes(A=@n18.Result, B=@n19.Material, Alpha=@n19.Material)
  n21 = VirtualTextureFeatureSwitch!(
   No=@n3
   Yes=StaticSwitchParameter(
    True=MaterialFunctionCall!(RVT=@n5, ByPass RVT=@n20) {materialFunction=/Clyde/Athena/Apollo/Maps/Landscape/VT/MaterialFunctions/EnableRVT}.Result
    False=@n20
   ) {defaultValue=true, group=RVT Base, parameterName=RVT Enable}
  )
  n22 = QualitySwitch(Default=@n3, Medium=@n21)
  GetMaterialAttributes(None=MaterialFunctionCall!(
    Material=ShadingPathSwitch(Default=@n21, Mobile=@n22)
    DebugInputAsBasecolor
    Non-RVT material
    DebugMaterialAs1Color=VectorParameter() {parameterName=DebugBiomeColor, defaultValue=<rgba 0 0.2359 0.474 1>}.RGB
   ) {materialFunction=/Game/Athena/Apollo/Environments/Landscape/Materials/Biome/CommonMaterials/MF_LS_DEBUG}.Result) {attributeGetTypes [
    #1
    #4
    #2
    #3
   ]}
  n23 = LandscapeLayerSample() {parameterName=BiomeA}
  n24 = LandscapeLayerSample() {parameterName=BiomeB}
  n25 = Saturate(None=Multiply(A=ScalarParameter() {defaultValue=5, parameterName=Mtn-ScreeplusForest-mod}, B=Max(A=@n23, B=@n24)))
  n26 = LandscapeLayerSample() {parameterName=BiomeC}
  n27 = Max(A=@n26, B=LandscapeLayerSample() {parameterName=BiomeD})
  n28 = Constant() {r=2}
  n29 = LandscapeLayerSample() {parameterName=Sand}
  n30 = Max(A=Max(A=LandscapeLayerSample() {parameterName=Gravel}, B=@n29), B=LandscapeLayerSample() {parameterName=Mud})
  n31 = LandscapeLayerSample() {parameterName=Asphalt}
  n32 = Saturate(None=Max(A=Multiply(A=@n28, B=@n30), B=Ceil(None=@n31)))
  n33 = OneMinus(None=Max(A=Max(A=@n25, B=@n27), B=Max(A=@n17.LandscapeGrassMeshMask, B=@n32)))
  MaterialExpressionReroute_55 = Reroute(None=Subtract(A=Subtract(A=@n33, B=@n19.GrassMask), B=LandscapeLayerSample() {parameterName=NoFoliage}))
  MaterialExpressionSubtract_8 = Subtract(A=Constant(), B=LandscapeLayerSample() {parameterName=NoFoliage})
  n34 = Constant() {r=2}
  n35 = Constant() {r=-0.5}
  n36 = Saturate(None=Add(A=Multiply(A=@n34, B=@n23), B=@n35))
  n37 = ScalarParameter() {defaultValue=-4, parameterName=Threshold_Scree}
  n38 = ScalarParameter() {defaultValue=1.6832, parameterName=Threshold_Scree2}
  n39 = ComponentMask(None=VertexNormalWS()) {b=true}
  n40 = Multiply(A=@n36, B=Saturate(None=LinearInterpolate(A=@n37, B=@n38, Alpha=@n39)))
  n41 = Constant()
  MaterialExpressionSubtract_3 = Subtract(A=ShadingPathSwitch(Default=@n40, Deferred=@n40, Forward=@n41, Mobile=@n41), B=LandscapeLayerSample() {parameterName=NoFoliage})
  n42 = VertexNormalWS()
  n43 = Saturate(None=LinearInterpolate(
    A=ScalarParameter() {defaultValue=-4, parameterName=Threshold_ScreeGrass}
    B=ScalarParameter() {defaultValue=1, parameterName=Threshold_ScreeGrass2}
    Alpha=ComponentMask(None=@n42) {b=true}
   ))
  MaterialExpressionSubtract_2 = Subtract(A=Subtract(A=Multiply(A=@n43, B=@n36), B=@n19.GrassMask), B=LandscapeLayerSample() {parameterName=NoFoliage})
  n44 = Add(A=@n35, B=Multiply(A=@n34, B=@n26))
  MaterialExpressionSubtract_1 = Subtract(A=Multiply(A=Saturate(None=@n44), B=OneMinus(None=@n17.LandscapeGrassMeshMask)), B=LandscapeLayerSample() {parameterName=NoFoliage})
  MaterialExpressionSubtract_0 = Subtract(A=Subtract(A=@n29, B=@n17.LandscapeGrassMeshMask), B=LandscapeLayerSample() {parameterName=NoFoliage})
  LandscapeGrassOutput!(
   Grass_Standard=@MaterialExpressionReroute_55
   Bush=@MaterialExpressionReroute_55
   FarmFlowers=@MaterialExpressionSubtract_8
   Scree=@MaterialExpressionSubtract_3
   Scree_Grass=@MaterialExpressionSubtract_2
   Heather=@MaterialExpressionSubtract_1
   SwampGrass=@MaterialExpressionSubtract_8
   SwampReeds=@MaterialExpressionSubtract_8
   Forest_Grass=@MaterialExpressionSubtract_8
   Beach=@MaterialExpressionSubtract_0
   Ocean=@MaterialExpressionSubtract_8
  ) {grassTypes [
    Grass_Standard {
     grassType=/Clyde/Landscape/GrassTypes/LandscapeGrassType_StandardGrass
     input {
      expression=&self:MaterialExpressionReroute_55
      inputName=None
      mask=0
      maskA=0
      maskB=0
      maskG=0
      maskR=0
      outputIndex=0
     }
    }
    Bush {
     grassType=/Clyde/Landscape/GrassTypes/Athena_Foliage_Bushes
     input {
      expression=&self:MaterialExpressionReroute_55
      inputName=None
      mask=0
      maskA=0
      maskB=0
      maskG=0
      maskR=0
      outputIndex=0
     }
    }
    FarmFlowers {
     grassType=/Clyde/Landscape/GrassTypes/LandscapeGrassType_FarmFlowers
     input {
      expression=&self:MaterialExpressionSubtract_8
      inputName=None
      mask=0
      maskA=0
      maskB=0
      maskG=0
      maskR=0
      outputIndex=0
     }
    }
    Scree {
     grassType=/Clyde/Landscape/GrassTypes/LandscapeGrassType_Scree
     input {
      expression=&self:MaterialExpressionSubtract_3
      inputName=None
      mask=0
      maskA=0
      maskB=0
      maskG=0
      maskR=0
      outputIndex=0
     }
    }
    Scree_Grass {
     grassType=/Clyde/Landscape/GrassTypes/LandscapeGrassType_ScreeGrass
     input {
      expression=&self:MaterialExpressionSubtract_2
      inputName=None
      mask=0
      maskA=0
      maskB=0
      maskG=0
      maskR=0
      outputIndex=0
     }
    }
    Heather {
     grassType=/Clyde/Landscape/GrassTypes/LandscapeGrassType_Heather
     input {
      expression=&self:MaterialExpressionSubtract_1
      inputName=None
      mask=0
      maskA=0
      maskB=0
      maskG=0
      maskR=0
      outputIndex=0
     }
    }
    SwampGrass {
     grassType=/Clyde/Landscape/GrassTypes/LandscapeGrassType_SwampGrass
     input {
      expression=&self:MaterialExpressionSubtract_8
      inputName=None
      mask=0
      maskA=0
      maskB=0
      maskG=0
      maskR=0
      outputIndex=0
     }
    }
    SwampReeds {
     grassType=/Clyde/Landscape/GrassTypes/LandscapeGrassType_SwampReeds
     input {
      expression=&self:MaterialExpressionSubtract_8
      inputName=None
      mask=0
      maskA=0
      maskB=0
      maskG=0
      maskR=0
      outputIndex=0
     }
    }
    Forest_Grass {
     grassType=/Clyde/Landscape/GrassTypes/LandscapeGrassType_ForestGrass
     input {
      expression=&self:MaterialExpressionSubtract_8
      inputName=None
      mask=0
      maskA=0
      maskB=0
      maskG=0
      maskR=0
      outputIndex=0
     }
    }
    Beach {
     grassType=/Clyde/Landscape/GrassTypes/LandscapeGrassType_Beach
     input {
      expression=&self:MaterialExpressionSubtract_0
      inputName=None
      mask=0
      maskA=0
      maskB=0
      maskG=0
      maskR=0
      outputIndex=0
     }
    }
    Ocean {
     grassType=/Clyde/Landscape/GrassTypes/LandscapeGrassType_Ocean
     input {
      expression=&self:MaterialExpressionSubtract_8
      inputName=None
      mask=0
      maskA=0
      maskB=0
      maskG=0
      maskR=0
      outputIndex=0
     }
    }
   ]}
  LandscapeLayerSample() {parameterName=BiomeD}
  LandscapeVisibilityMask()
  n45 = GetMaterialAttributes(None=@n20) {attributeGetTypes [
    #1
    #2
    #3
    #4
   ]}
  RuntimeVirtualTextureOutput!(
   BaseColor=MaterialFunctionCall!(BaseColor=@n45.Base Color) {materialFunction=/Clyde/Athena/Apollo/Environments/Landscape/Materials/Biome/CommonMaterials/MF_ReduceRTV-CompressionArtifacts}.BaseColor
   Specular=@n45.Specular
   Roughness=@n45.Specular
   Normal=@n45.Specular
   WorldHeight=ComponentMask(None=WorldPosition().XYZ) {b=true}
   Opacity
   Mask
   Displacement
   Mask4
  )
  VectorParameter() {parameterName=Biome-CliffDiv, defaultValue=<rgba 1 1 1 1>}
  VectorParameter() {parameterName=Biome-CliffColor, defaultValue=<rgba 1 0.96 0.9 1>}
 ]
}
