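# Common landscape material for Apollo "W3W4" terrain, blending painted ground layers, Forest and FarmGrass biomes, slope-based cliffs and procedural snow in world-space normals.
# The result writes to a runtime virtual texture, optionally sampled back through RVT switches.
# A simpler mobile layer function covers mobile quality levels other than Medium and platforms without virtual texturing.
# Grass output derives foliage from layer, slope, snow and height masks, suppressed by NoFoliage.
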Material /Clyde/Athena/Apollo/Environments/Landscape/Materials/Biome/LandscapeBiomeMaterials/M_CommonLandscape_W3W4
graph {
 properties {
  bTangentSpaceNormal=false
  bUseMaterialAttributes=true
  bUsedWithNanite=true
  bUsedWithVirtualHeightfieldMesh=true
  pixelDepthOffsetMode=PDOM_Legacy
  usedShadingModels=Default Lit
 }
 nodes [
  Constant() {r=-4400}
  n1 = LandscapeLayerSample() {parameterName=Gravel}
  n2 = LandscapeLayerSample() {parameterName=Mud}
  n3 = Max(A=Multiply(A=Max(A=@n1, B=@n2), B=Constant() {r=2}), B=Multiply(A=LandscapeLayerSample() {parameterName=Sand}, B=Constant() {r=5}))
  n4 = Ceil(None=LandscapeLayerSample() {parameterName=Asphalt})
  n5 = MaterialFunctionCall!(
   UseBiomeColorMask_1
   BiomeMask_1
   GrassBorderColorMask_1=@n11.MaskForGrassBorderColor
   HeightmapFromCliff_1=@n11.HeightmapForGrass
   Grass-EdgeBiomeColorB_1
   Grass-EdgeBiomeColorA_1
   GravelBiomeColorDiv_1
   GravelBiomeColor_1
   MudBiomeColorDiv_1
   MudBiomeColor_1
   MudBiomeRoughMod_1
   SandBiomeColorDiv_1
   SandBiomeColor_1
  ) {materialFunction=/Clyde/Athena/Apollo/Environments/Landscape/Materials/Biome/CommonMaterials/MF_CommonLandscapeLayers}
  n6 = MaterialFunctionCall!(
   Material 1=@n5.Material
   Height Texture 1=@n5.Heightmap
   Material 2=@n5.Material
   Height Texture 2=@n5.Heightmap
   Transistion Phase=LandscapeLayerSample() {parameterName=NoFoliage}
   UseContrast
   Contrast
  ) {materialFunction=/Clyde/Athena/Apollo/Environments/Landscape/Materials/Biome/CommonMaterials/MF_BlendLandscapeLayers}
  n7 = MaterialFunctionCall!(UseBiomeColorMask, BiomeColorMask, BiomeColorDiv, BiomeColor) {materialFunction=/Clyde/Athena/Apollo/Environments/Landscape/Materials/Biome/UniqueBiomeMaterialFunctions/MF_Biome_Forest}
  n8 = MaterialFunctionCall(
   Material 1=@n6.BlendedMaterial
   Height Texture 1=@n6.BlendedMaterial
   Material 2=@n7.Material
   Height Texture 2=@n7.Material
   Transistion Phase=LandscapeLayerSample() {parameterName=BiomeB}
  ) {materialFunction=&MF_BlendLandscapeLayers}
  n9 = MaterialFunctionCall!(
   GrassHeightmap=@n5.GrassNoiseMask_1
   GrassNoiseMask=@n5.GrassHeightmap_1
   GrassBaseColorDetail=@n5.GrassBaseColorDetail_1
   GrassMaterial=@n5.Material
  ) {materialFunction=/Clyde/Athena/Apollo/Environments/Landscape/Materials/Biome/UniqueBiomeMaterialFunctions/MF_Biome_FarmGrass}
  n10 = MaterialFunctionCall(
   Material 1=@n8.BlendedMaterial
   Height Texture 1=@n8.BlendedMaterial
   Material 2=@n9.Material
   Height Texture 2=@n9.Material
   Transistion Phase=LandscapeLayerSample() {parameterName=BiomeC}
  ) {materialFunction=&MF_BlendLandscapeLayers}
  n11 = MaterialFunctionCall!(BiomeColorMask, Material=@n10.BlendedMaterial, Heightmap=@n10.BlendedMaterial, UseBiomeColorMask, BiomeColorDiv, BiomeColor) {materialFunction=/Clyde/Athena/Apollo/Environments/Landscape/Materials/Biome/CommonMaterials/MF_SlopeBlendCliff}
  n12 = Max(A=@n11.LandscapeGrassMeshMask, B=Saturate(None=Max(A=@n3, B=@n4)))
  n13 = MaterialFunctionCall!(Material=@n11.Material) {materialFunction=/Clyde/Athena/Apollo/Maps/Landscape/VT/MaterialFunctions/TransformTangentToWorld}
  n14 = MaterialFunctionCall!(MaterialIn=@n13.Result) {materialFunction=/Clyde/Athena/Apollo/Environments/Landscape/Materials/Biome/UniqueBiomeMaterialFunctions/MF_ProceduralSnow}
  MaterialExpressionSubtract_9 = Subtract(A=Subtract(A=OneMinus(None=@n12), B=@n14.MaterialMask), B=LandscapeLayerSample() {parameterName=NoFoliage})
  n15 = Constant() {r=2}
  n16 = LandscapeLayerSample() {parameterName=BiomeC}
  MaterialExpressionSubtract_10 = Subtract(A=Add(A=Constant() {r=-0.5}, B=Multiply(A=@n15, B=@n16)), B=LandscapeLayerSample() {parameterName=NoFoliage})
  MaterialExpressionSubtract_6 = Subtract(A=Constant(), B=LandscapeLayerSample() {parameterName=NoFoliage})
  n17 = Constant()
  n18 = ComponentMask(None=WorldPosition().XYZ) {b=true}
  n19 = Subtract(A=@n18, B=Constant() {r=-5100})
  n20 = Constant() {r=1}
  n21 = OneMinus(None=Saturate(None=Divide(A=@n19, B=@n20)))
  n22 = Subtract(A=Subtract(A=LandscapeLayerSample() {parameterName=Sand}, B=@n11.LandscapeGrassMeshMask), B=@n21)
  MaterialExpressionSubtract_7 = Subtract(A=ShadingPathSwitch(Default=@n22, Deferred=@n22, Forward=@n17, Mobile=@n17), B=LandscapeLayerSample() {parameterName=NoFoliage})
  n23 = Constant()
  n24 = Divide(A=Subtract(A=@n18, B=Constant() {r=-4600}), B=@n20)
  n25 = Subtract(A=OneMinus(None=Saturate(None=@n24)), B=@n21)
  MaterialExpressionReroute_11 = Reroute(None=Subtract(A=ShadingPathSwitch(Default=@n25, Deferred=@n25, Forward=@n23, Mobile=@n23), B=LandscapeLayerSample() {parameterName=NoFoliage}))
  LandscapeGrassOutput!(
   Grass_Standard=@MaterialExpressionSubtract_9
   Bush=@MaterialExpressionSubtract_9
   FarmFlowers=@MaterialExpressionSubtract_10
   Scree=@MaterialExpressionSubtract_6
   Scree_Grass=@MaterialExpressionSubtract_6
   Heather=@MaterialExpressionSubtract_6
   SwampGrass=@MaterialExpressionSubtract_6
   SwampReeds=@MaterialExpressionSubtract_6
   Forest_Grass=@MaterialExpressionSubtract_6
   Beach=@MaterialExpressionSubtract_7
   Ocean=@MaterialExpressionReroute_11
  ) {grassTypes [
    Grass_Standard {
     grassType=/Clyde/Landscape/GrassTypes/Season2/LandscapeGrassType_S2_StandardGrass
     input {
      expression=&self:MaterialExpressionSubtract_9
      inputName=None
      mask=0
      maskA=0
      maskB=0
      maskG=0
      maskR=0
      outputIndex=0
     }
    }
    Bush {
     grassType=/Clyde/Landscape/GrassTypes/Season2/Athena_Foliage_S2_Bushes
     input {
      expression=&self:MaterialExpressionSubtract_9
      inputName=None
      mask=0
      maskA=0
      maskB=0
      maskG=0
      maskR=0
      outputIndex=0
     }
    }
    FarmFlowers {
     grassType=/Clyde/Landscape/GrassTypes/Season2/LandscapeGrassType_S2_FarmFlowers
     input {
      expression=&self:MaterialExpressionSubtract_10
      inputName=None
      mask=0
      maskA=0
      maskB=0
      maskG=0
      maskR=0
      outputIndex=0
     }
    }
    Scree {
     grassType=/Clyde/Landscape/GrassTypes/Season2/LandscapeGrassType_S2_Scree
     input {
      expression=&self:MaterialExpressionSubtract_6
      inputName=None
      mask=0
      maskA=0
      maskB=0
      maskG=0
      maskR=0
      outputIndex=0
     }
    }
    Scree_Grass {
     grassType=/Clyde/Landscape/GrassTypes/Season2/LandscapeGrassType_S2_ScreeGrass
     input {
      expression=&self:MaterialExpressionSubtract_6
      inputName=None
      mask=0
      maskA=0
      maskB=0
      maskG=0
      maskR=0
      outputIndex=0
     }
    }
    Heather {
     grassType=/Clyde/Landscape/GrassTypes/Season2/LandscapeGrassType_S2_Heather
     input {
      expression=&self:MaterialExpressionSubtract_6
      inputName=None
      mask=0
      maskA=0
      maskB=0
      maskG=0
      maskR=0
      outputIndex=0
     }
    }
    SwampGrass {
     grassType=/Clyde/Landscape/GrassTypes/Season2/LandscapeGrassType_S2_SwampGrass
     input {
      expression=&self:MaterialExpressionSubtract_6
      inputName=None
      mask=0
      maskA=0
      maskB=0
      maskG=0
      maskR=0
      outputIndex=0
     }
    }
    SwampReeds {
     grassType=/Clyde/Landscape/GrassTypes/Season2/LandscapeGrassType_S2_SwampReeds
     input {
      expression=&self:MaterialExpressionSubtract_6
      inputName=None
      mask=0
      maskA=0
      maskB=0
      maskG=0
      maskR=0
      outputIndex=0
     }
    }
    Forest_Grass {
     grassType=/Clyde/Landscape/GrassTypes/Season2/LandscapeGrassType_S2_ForestGrass
     input {
      expression=&self:MaterialExpressionSubtract_6
      inputName=None
      mask=0
      maskA=0
      maskB=0
      maskG=0
      maskR=0
      outputIndex=0
     }
    }
    Beach {
     grassType=/Clyde/Landscape/GrassTypes/Season2/LandscapeGrassType_S2_Beach
     input {
      expression=&self:MaterialExpressionSubtract_7
      inputName=None
      mask=0
      maskA=0
      maskB=0
      maskG=0
      maskR=0
      outputIndex=0
     }
    }
    Ocean {
     grassType=/Clyde/Landscape/GrassTypes/Season2/LandscapeGrassType_S2_Ocean
     input {
      expression=&self:MaterialExpressionReroute_11
      inputName=None
      mask=0
      maskA=0
      maskB=0
      maskG=0
      maskR=0
      outputIndex=0
     }
    }
   ]}
  n26 = GetMaterialAttributes(None=MaterialFunctionCall!(UseSnow=StaticBool() {value=true}) {materialFunction=/Clyde/Athena/Apollo/Environments/Landscape/Materials/Biome/CommonMaterials/Mobile/MF_Mobile_CommonLandscapeLayers}.Material) {attributeGetTypes [#1]}
  MaterialExpressionConstant_13 = Constant()
  n27 = LandscapeLayerBlend!(
   Layer Gravel=@MaterialExpressionConstant_13
   Layer Mud=@MaterialExpressionConstant_13
   Layer Sand=@MaterialExpressionConstant_13
   Layer Asphalt=@MaterialExpressionConstant_13
   Layer NoFoliage=@MaterialExpressionConstant_13
  ) {layers [
    {
     blendType=LB_WeightBlend
     constHeightInput=0
     layerName=Gravel
     previewWeight=0
     constLayerInput=<0 0 0>
     heightInput {expression=null, inputName=None, mask=0, maskA=0, maskB=0, maskG=0, maskR=0, outputIndex=0}
     layerInput {
      expression=&self:MaterialExpressionConstant_13
      inputName=None
      mask=0
      maskA=0
      maskB=0
      maskG=0
      maskR=0
      outputIndex=0
     }
    }
    {
     blendType=LB_WeightBlend
     constHeightInput=0
     layerName=Mud
     previewWeight=0
     constLayerInput=<0 0 0>
     heightInput {expression=null, inputName=None, mask=0, maskA=0, maskB=0, maskG=0, maskR=0, outputIndex=0}
     layerInput {
      expression=&self:MaterialExpressionConstant_13
      inputName=None
      mask=0
      maskA=0
      maskB=0
      maskG=0
      maskR=0
      outputIndex=0
     }
    }
    {
     blendType=LB_WeightBlend
     constHeightInput=0
     layerName=Sand
     previewWeight=0
     constLayerInput=<0 0 0>
     heightInput {expression=null, inputName=None, mask=0, maskA=0, maskB=0, maskG=0, maskR=0, outputIndex=0}
     layerInput {
      expression=&self:MaterialExpressionConstant_13
      inputName=None
      mask=0
      maskA=0
      maskB=0
      maskG=0
      maskR=0
      outputIndex=0
     }
    }
    {
     blendType=LB_WeightBlend
     constHeightInput=0
     layerName=Asphalt
     previewWeight=0
     constLayerInput=<0 0 0>
     heightInput {expression=null, inputName=None, mask=0, maskA=0, maskB=0, maskG=0, maskR=0, outputIndex=0}
     layerInput {
      expression=&self:MaterialExpressionConstant_13
      inputName=None
      mask=0
      maskA=0
      maskB=0
      maskG=0
      maskR=0
      outputIndex=0
     }
    }
    {
     blendType=LB_WeightBlend
     constHeightInput=0
     layerName=NoFoliage
     previewWeight=0
     constLayerInput=<0 0 0>
     heightInput {expression=null, inputName=None, mask=0, maskA=0, maskB=0, maskG=0, maskR=0, outputIndex=0}
     layerInput {
      expression=&self:MaterialExpressionConstant_13
      inputName=None
      mask=0
      maskA=0
      maskB=0
      maskG=0
      maskR=0
      outputIndex=0
     }
    }
   ]}
  n28 = SetMaterialAttributes!(MaterialAttributes=@n26.MaterialAttributes, Base Color=Add(A=@n26.Base Color, B=Multiply(A=@n27) {constB=0})) {attributeSetTypes [#1]}
  n29 = RuntimeVirtualTextureSampleParameter() {
   group=RVT
   materialType=BaseColor_Normal_Specular_YCoCg
   parameterName=RVT Sample Target
   virtualTexture=/Game/Athena/Apollo/RVT/RVT_Apollo_Landscape_Shark
  }
  n30 = SetMaterialAttributes!(MaterialAttributes, Base Color=@n29.BaseColor, Specular=@n29.BaseColor, Roughness=@n29.BaseColor, Normal=@n29.BaseColor) {attributeSetTypes [
    #1
    #2
    #3
    #4
   ]}
  n31 = BlendMaterialAttributes(A=@n13.Result, B=@n14.Material, Alpha=@n14.Material) {desc="Vertex blend type altered!", vertexAttributeBlendType=UseA}
  n32 = VirtualTextureFeatureSwitch!(
   No=@n28
   Yes=StaticSwitchParameter(
    True=MaterialFunctionCall!(RVT=@n30, ByPass RVT=@n31) {materialFunction=/Game/Athena/Apollo/Maps/Landscape/VT/MaterialFunctions/EnableRVT}.Result
    False=@n31
   ) {group=RVT Base, parameterName=RVT Enable}
  )
  MaterialFunctionCall!(Material=ShadingPathSwitch(Default=@n32, Mobile=QualitySwitch(Default=@n28, Medium=@n32)), DebugAsBasecolor, DebugBiomeColor) {materialFunction=/Clyde/Athena/Apollo/Environments/Landscape/Materials/Biome/CommonMaterials/MF_LS_DEBUG}
  n33 = GetMaterialAttributes(None=@n31) {attributeGetTypes [
    #1
    #2
    #3
    #4
   ]}
  RuntimeVirtualTextureOutput!(
   BaseColor=MaterialFunctionCall!(BaseColor=@n33.Base Color) {materialFunction=/Game/Athena/Apollo/Environments/Landscape/Materials/Biome/CommonMaterials/MF_ReduceRTV-CompressionArtifacts}.BaseColor
   Specular=@n33.Specular
   Roughness=@n33.Specular
   Normal=@n33.Specular
   WorldHeight=Add(A=BreakMaterialAttributes!(Attr=@n33.MaterialAttributes).Displacement, B=WorldPosition().Z)
   Opacity
   Mask
   Displacement
   Mask4
  )
  VectorParameter() {parameterName=DebugBiomeColor, defaultValue=<rgba 0.0445 1 0 1>}
 ]
}
